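# Niagara system for the jet burst of the VinderTech Noble male glider chute, drawing timed flares and flame meshes at the exhausts.
# Particles follow sockets on the user-supplied skeletal mesh, with flame meshes oriented along the sampled socket rotation.
# Emitter-age curves shape the burst's size, alpha and spawning, alongside system thruster power and front-end versus gameplay modifiers.
# Several flame emitter variants, including the ForwardDot-driven idle/active one, are disabled.
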NiagaraSystem /BRCosmetics/Gadgets/Assets/VinderTech_GliderChute/Glider_Noble_Male/FX/NS_Glider_Noble_JetBurst
dependencies {
 usedDataInterfaces [
  &NiagaraDataInterfaceVectorCurve
  &NiagaraDataInterfaceCurve
  &NiagaraDataInterfaceColorCurve
  &NiagaraDataInterfaceVector2DCurve
  /Script/Niagara.NiagaraDataInterfaceSimpleCounter
  &NiagaraDataInterfaceArrayFloat3
  &NiagaraDataInterfaceArrayFloat
 ]
 usedDynamicInputs [
  &Boolean_AndOperation
  &HueShiftLinearColor
  &Divide_Float
 ]
 usedModules [
  &self:SystemSpawnScript.SystemScriptSource.SystemScriptGraph.NiagaraNodeAssignment_0.SetVariables_#1
  &FN_Integrator
  &self:SystemSpawnScript.SystemScriptSource.SystemScriptGraph.NiagaraNodeAssignment_1.SetVariables_#2
  &SystemState
  &self:Set_Current_Exhaust
  &self:Set_Trail_Length
  &self:Set_Trail_Opacity
  &Timeline
  &self:SystemSpawnScript.SystemScriptSource.SystemScriptGraph.NiagaraNodeAssignment_7.SetVariables_#3
  &self:SystemSpawnScript.SystemScriptSource.SystemScriptGraph.NiagaraNodeAssignment_3.SetVariables_#4
 ]
}
emitters {
 Big_Old_Flares {
  emitterUpdateScript [
   &EmitterState(Life Cycle Mode=::NewEnumerator1, Loop Duration=2, UseLoopDelay=false)
   &SpawnRate(
    Spawn Probability=&FloatFromCurve(
     FloatCurve=DataInterface() {propertyValues {Curve {keys [
         {tangentMode=RCTM_User, time=0, value=1}
         {arriveTangent=-0.0405, interpMode=RCIM_Cubic, leaveTangent=-0.0405, tangentMode=RCTM_User, time=0.7009, value=1}
         {arriveTangent=-1.5834, interpMode=RCIM_Cubic, leaveTangent=-1.5834, tangentMode=RCTM_User, time=3, value=0}
        ]}}}
     CurveIndex=$Emitter.Age
    )
    SpawnRate=100
    Use Spawn Probability=true
   )
   &SpawnRate(SpawnGroup=2, SpawnRate=100, Use Spawn Probability=false) {enabled=false, module=SpawnRate001}
  ]
  particleSpawnScript [
   &InitializeParticle(
    Color=&MakeLinearColorFromVectorAndFloat(
     "Vector (RGB)"=&MakeVectorFromLinearColorRGB(Color=<rgba 0 0.1667 1 1>)
     "Float (Alpha)"=&FloatFromCurve(
      FloatCurve=DataInterface() {propertyValues {Curve {keys [<at 0 1>, {arriveTangent=-0.4545, interpMode=RCIM_Cubic, leaveTangent=-0.4545, time=1, value=1}, <at 2.2 0>]}}}
      CurveIndex=$Emitter.Age
      Scale Curve=0.1
     )
    )
    Lifetime=0.2
    Sprite Size Mode=::NewEnumerator3
    Uniform Sprite Size=&FloatFromCurve(
     FloatCurve=DataInterface() {propertyValues {Curve {keys [
         {tangentMode=RCTM_User, time=0, value=1}
         {arriveTangent=-0.0405, interpMode=RCIM_Cubic, leaveTangent=-0.0405, tangentMode=RCTM_User, time=0.2, value=1}
         {arriveTangent=-0.1588, interpMode=RCIM_Cubic, leaveTangent=-0.1588, tangentMode=RCTM_User, time=1, value=0.8}
         {arriveTangent=-1.3481, interpMode=RCIM_Cubic, leaveTangent=-1.3481, tangentMode=RCTM_User, time=2, value=0}
        ]}}}
     CurveIndex=$Emitter.Age
     Scale Curve=100
    )
    UsePositionOffset=true
    Write Material Random=true
   )
   &SkeletalMeshLocation(
    Bone / Socket Index=2
    MaskSpawnGroup=true
    Mesh Sampling Type=::NewEnumerator5
    Orientation Sampling=::NewEnumerator1
    Sampled Position Offset=<5 0 0>
    Skeletal Mesh=$User.Skeletal Mesh
    Socket Sampling Mode=::NewEnumerator5
    Spawn Group Mask=1
   ) {enabled=false}
   &FN_SetSpawnGroup(Number of Spawn Groups=2) {enabled=false}
   /SetVariables(Particles.SpriteSize=&MaskVector2dBySpawnGroup(
     Masked Vector2D=&Multiply_Vector2D(A=$Particles.SpriteSize, B=<4 4>)
     Passthrough Vector2D=&Multiply_Vector2D(A=$Particles.SpriteSize, B=<1 1>)
     Spawn Group Mask=0
    )) {enabled=false, isSetParameters=true}
  ]
  particleUpdateScript [
   &ParticleState()
   /SetVariables() {enabled=false, isSetParameters=true}
   &SkeletalMeshLocation(
    Bone / Socket Index=&RandomRangeInt(Minimum=0, Maximum=1)
    MaskSpawnGroup=true
    Mesh Sampling Type=::NewEnumerator5
    Orientation Sampling=::NewEnumerator1
    Sampled Position Offset=<1 0 0>
    Skeletal Mesh=$User.Skeletal Mesh
    Socket Sampling Mode=::NewEnumerator5
   ) {module=SkeletalMeshLocation001}
   &SpriteFacingAndAlignment(
    Sprite Facing=&RotateVectorByQuaternion(Vector To Rotate=<-1 0 0>, Quaternion=$Particles.SkeletalMeshLocation001.SampledRotation)
    WriteAlignment=false
    WriteFacing=true
   ) {enabled=false}
   /SetVariables(
    Particles.Color_Particle_B:LinearColor=<rgba 0 3.3333 20 0.1>
    Particles.SpriteSizeAlt:Vector2f=&Multiply_Vector2D(
     A=$Particles.Initial.SpriteSize
     B=&Vector2DFromCurve(Vector2Curve=DataInterface() {propertyValues {XCurve {keys [<at 0 1>,<at 1 1>]}, YCurve {keys [<at 0 1>,<at 1 1>]}}}, ScaleCurve=<0.4 0.4>)
    )
   ) {isSetParameters=true}
   &CameraOffset(Camera Offset Amount=20)
  ]
  rendererClasses [&NiagaraSpriteRendererProperties]
  renderers [
   {rendererClass=&NiagaraSpriteRendererProperties}
   {rendererClass=&NiagaraSpriteRendererProperties, rendererIndex=1}
  ]
 }
 Empty001 {
  bEnabled=false
  emitterUpdateScript [
   &EmitterState(Life Cycle Mode=::NewEnumerator1, Loop Duration=1, UseLoopDelay=false)
   &SpawnRate(SpawnRate=16, Use Spawn Probability=false)
  ]
  particleSpawnScript [&InitializeParticle(
    Color Mode=::NewEnumerator0
    Lifetime Max=1
    Lifetime Min=0.5
    Lifetime Mode=::NewEnumerator1
    Mesh Scale=<0.12 0.12 0.1>
    Mesh Scale Mode=::NewEnumerator4
    Sprite Size Mode=::NewEnumerator3
    Uniform Sprite Size=100
    UsePositionOffset=false
    Write Material Random=true
   )]
  particleUpdateScript [&ParticleState()]
  rendererClasses [&NiagaraSpriteRendererProperties]
  renderers [{rendererClass=&NiagaraSpriteRendererProperties}]
 }
 Flame_Mesh {
  bEnabled=false
  emitterUpdateScript [
   &EmitterState()
   &SpawnRate(SpawnRate=40, Use Spawn Probability=false)
  ]
  particleSpawnScript [
   /Niagara/Modules/Spawn/Initialization/InitializeParticle(
    Color=<rgba 10 6.25 1 0.25>
    Lifetime=0.2
    Mesh Scale=&FN_FrontEndVector3Modifier(
     Gameplay=&Multiply_VectorByFloat(Vector=&RandomRangeVector(Minimum=<0.2 0.125 0.3>, Maximum=<0.3 0.125 0.3>), Float=$User.ForwardDot)
     Front End=&RandomRangeVector(Minimum=<0.2 0.1 0.2>, Maximum=<0.3 0.1 0.2>)
    )
    Module.Write Color=true
    Module.Write Lifetime=true
    Module.Write Mass=false
    Module.Write Material Random=false
    Module.Write Position=false
    Module.Write Scale=true
    Module.Write SpriteRotation=false
    Module.Write SpriteSize=false
    Position=&SimulationPosition()
    Sprite Size=<5 5>
   )
   &FN_SetSpawnGroup(Number of Spawn Groups=2)
   /SetVariables(
    Particles.Color_3:LinearColor=&MultiplyLinearColor(
     Linear Color=<rgba 1.5 0.375 0.15 2.5>
     Scale Factor=&MakeVector4_FromFloat(Float=&RandomRangeFloat(Minimum=1, Maximum=5))
    )
    Particles.Scale_3:Vector3f=&RandomRangeVector(Minimum=<0.04 0.095 0.5>, Maximum=<0.04 0.095 1>)
   ) {isSetParameters=true}
   &DynamicMaterialParameters(
    Index 0 Param 1=5
    Index 0 Param 2=5
    Module.Param0WriteEnabled=true
    Module.Param1WriteEnabled=true
    Module.Param2WriteEnabled=true
    Module.Param3WriteEnabled=true
   )
   /SetVariables(Particles.DynamicMaterialParameter_3:Vector4f=&MakeVector4(X=&RandomRangeFloat(Minimum=0, Maximum=1), Y=0, Z=0, W=0)) {isSetParameters=true}
   /Niagara/Modules/Spawn/Orientation/V2/InitialMeshOrientation(
    Orientation Axis:Vector3f=/Game/Effects/Niagara/DynamicInputs/Water/FN_PickVectorFromSpawnGroupCutoff(
     Under Cutoff Vector:Vector3f=<1 0.07 0>
     Over Cutoff Vector:Vector3f=<1 -0.07 0>
     SpawnGroup:NiagaraInt32=$Particles.SpawnGroup
     Cutoff:NiagaraInt32=1
    )
    Orientation Vector=<0 1 0>
    Rotation=&FN_FrontEndVector3Modifier(Gameplay=<0 0 0>, Front End=<0 0.04 0>)
    Use Orientation Vector=true
    Use Rotation Vector=true
   ) {enabled=false, module=InitialMeshOrientation001}
   &SkeletalMeshLocation(
    Apply Sampled Orientation to Mesh Particles=true
    Bone / Socket Index=$Particles.SpawnGroup
    Inherit Velocity=false
    MaskSpawnGroup=false
    Mesh Sampling Type=::NewEnumerator5
    Sampled Position Offset=$System.OffsetLoopFromEngineBone
    Sampled Position Offset Coordinate Space:ENiagaraMeshTransforms=::NewEnumerator3
    Skeletal Mesh=$User.Skeletal Mesh
    Socket Sampling Mode=::NewEnumerator5
   ) {module=SkeletalMeshLocation001}
   /Niagara/Modules/Spawn/Orientation/V3/InitialMeshOrientation(
    Mesh Axis To Orient=&RotateVectorByQuaternion(Vector To Rotate=<1 0 0>, Quaternion=&MakeQuatFromAxisAngle(Axis=<0 1 0>, AngleInDegrees=-15))
    Mesh Orientation Mode=::NewEnumerator2
    Orientation Coordinate Space=::Local
    Orientation Vector=<0 1 0>
    Use Rotation Vector=false
   )
  ]
  particleUpdateScript [
   &ParticleState()
   &SkeletalMeshLocation(
    Apply Sampled Orientation to Mesh Particles=true
    Bone / Socket Index=$Particles.SpawnGroup
    Inherit Velocity=false
    Interpolate Mesh Sampling=false
    MaskSpawnGroup=false
    Mesh Sampling Type=::NewEnumerator5
    Sampled Position Offset=&Add_Vector(
     A=$System.OffsetLoopFromEngineBone
     B=&FN_FrontEndVector3Modifier(Gameplay=<0 3 -3>, Front End=<0 -10 -8>)
    )
    Sampled Position Offset Coordinate Space:ENiagaraMeshTransforms=::NewEnumerator3
    Skeletal Mesh=$User.Skeletal Mesh
    Socket Sampling Mode=::NewEnumerator5
   ) {module=SkeletalMeshLocation002}
   /SetVariables(
    Particles.MeshOrientation_3:Quat4f=&MultiplyQuaternion(
     Quaternion A=$Output.SkeletalMeshLocation002.SampledRotation
     Quaternion B=&LerpQuaternion(
      Quaternion A:Quat4f=<0 -0.4 0 0.892>
      Quaternion B:Quat4f=<0 -0.7 0 0.892>
      Lerp Factor:NiagaraFloat=&Multiply_Float(A=$Particles.NormalizedAge, B=$User.ForwardDot)
      Enable Spherical Linear Interpolation:NiagaraBool=false
     )
    )
    Particles.Position_3:NiagaraPosition=$Particles.Position
   ) {isSetParameters=true}
   &SkeletalMeshLocation(
    Apply Sampled Orientation to Mesh Particles=true
    Bone / Socket Index=$Particles.SpawnGroup
    Inherit Velocity=false
    Interpolate Mesh Sampling=false
    MaskSpawnGroup=false
    Mesh Sampling Type=::NewEnumerator5
    Sampled Position Offset=&Add_Vector(
     A=$System.OffsetLoopFromEngineBone
     B=&FN_FrontEndVector3Modifier(
      Gameplay=&Lerp_Vector(
       A=<0 2 0>
       B=<0 10 0>
       Alpha=&FloatFromCurve(FloatCurve=DataInterface() {propertyValues {Curve {keys [
            {arriveTangent=0.0743, interpMode=RCIM_Cubic, leaveTangent=0.0743, tangentMode=RCTM_User, time=0, value=0}
            {arriveTangent=0.8517, interpMode=RCIM_Cubic, leaveTangent=0.8517, tangentMode=RCTM_User, time=0.7665, value=0.1865}
            {arriveTangent=5.2956, interpMode=RCIM_Cubic, leaveTangent=5.2956, tangentMode=RCTM_User, time=1, value=1}
           ]}}})
      )
      Front End=<0 -3.6605 -5.1933>
     )
    )
    Sampled Position Offset Coordinate Space:ENiagaraMeshTransforms=::NewEnumerator3
    Skeletal Mesh=$User.Skeletal Mesh
    Socket Sampling Mode=::NewEnumerator5
   )
   &ScaleMeshSize(Scale Factor=&MakeVector(
     X=&FloatFromCurve(FloatCurve=DataInterface() {propertyValues {Curve {keys [<at 0 0.4 RCIM_Cubic>,<at 1 1 RCIM_Cubic>]}}})
     Y=&FloatFromCurve(FloatCurve=DataInterface() {propertyValues {Curve {keys [<at 0 0.3 RCIM_Cubic>, {arriveTangent=-0.3, interpMode=RCIM_Cubic, leaveTangent=-0.3, time=0.25, value=1}, <at 1 0 RCIM_Cubic>]}}})
     Z=&FloatFromCurve(FloatCurve=DataInterface() {propertyValues {Curve {keys [<at 0 0.3 RCIM_Cubic>, {arriveTangent=-0.3, interpMode=RCIM_Cubic, leaveTangent=-0.3, time=0.25, value=1}, <at 1 0 RCIM_Cubic>]}}})
    ))
   /SetVariables(
    Particles.Color_3:LinearColor=&MultiplyLinearColor(
     Linear Color=$Particles.Color_3
     Scale Factor=&MakeFloat4FromVectorAndFloat(
      "Vector (RGB)"=<1 1 1>
      "Float (Alpha)"=&Lerp_Float(A=0.5, B=&Lerp_Float(A=5, B=10, Alpha=$User.ForwardDot), Alpha=$Particles.NormalizedAge)
     )
    )
    Particles.Scale_3:Vector3f=&Multiply_Vector(
     A=$Particles.Initial.Scale_3
     B=&Lerp_Vector(
      A=<0.5 1 0.5>
      B=&Lerp_Vector(A=<0.25 0.1 1>, B=<1 0.9 1>, Alpha=$User.ForwardDot)
      Alpha=$Particles.NormalizedAge
     )
    )
   ) {isSetParameters=true}
   /SetVariables(Particles.Scale_4:Vector3f=&Multiply_Vector(
     A=$Particles.Initial.Scale_3
     B=&Lerp_Vector(
      A=<1 1.1 0.5>
      B=&Lerp_Vector(A=<0.25 0.75 1>, B=<1 0.9 1>, Alpha=$User.ForwardDot)
      Alpha=&Multiply_Float(A=1, B=$Particles.NormalizedAge)
     )
    )) {isSetParameters=true}
   /SetVariables(
    Particles.VisibilityActive:NiagaraInt32=&MakeCustomIntFromBool(
     Bool=&SetBoolByFloatComparison(A=$User.ForwardDot, B=0.1, Comparison Type=::NewEnumerator6)
     True INT=1
     False INT=0
    )
    Particles.VisibilityIdle:NiagaraInt32=&MakeCustomIntFromBool(Bool=&SetBoolByFloatComparison(A=$User.ForwardDot, B=0.1), True INT=1, False INT=0)
   ) {isSetParameters=true}
  ]
  rendererClasses [&NiagaraMeshRendererProperties]
  renderers [
   {rendererClass=&NiagaraMeshRendererProperties, rendererIndex=0}
   {rendererClass=&NiagaraMeshRendererProperties, rendererIndex=1}
   {rendererClass=&NiagaraMeshRendererProperties, rendererIndex=2}
  ]
 }
 Flame_Mesh001 {
  bEnabled=false
  emitterUpdateScript [
   &EmitterState()
   &SpawnRate(
    SpawnRate=&Lerp_Float(A=80, B=140, Alpha=&FN_FrontEndFloatModifier(Gameplay=$System.ThrusterPower, Front End=1))
    Use Spawn Probability=false
   )
   &SpawnRate(Spawn Probability=$System.GenerateForwardBurst, SpawnRate=140, Use Spawn Probability=true) {enabled=false, module=SpawnRate001}
  ]
  particleSpawnScript [
   &SkeletalMeshLocation(
    Bone / Socket Index=&RandomRangeInt(Minimum=0, Maximum=1)
    MaskSpawnGroup=false
    Mesh Sampling Type=::NewEnumerator5
    Orientation Sampling=::NewEnumerator1
    Sampled Position Offset=<0 0 1.5>
    Skeletal Mesh=$User.Skeletal Mesh
    Socket Sampling Mode=::NewEnumerator5
   ) {module=SkeletalMeshLocation001}
   /SetVariables(
    Particles.PlacementPicker:NiagaraInt32=/Niagara/DynamicInputs/Divide/Divide_Int(
     A:NiagaraInt32=&ModuloInt(
      A=/Niagara/DynamicInputs/DataInterface/GenerateIntFromCounter(SimpleCounter:NiagaraDataInterfaceSimpleCounter=DataInterface() {propertyValues {GpuSyncMode=None, InitialValue=0}})
      B=4
     )
     B:NiagaraInt32=2
    )
    Particles.SpawnGroup=&ModuloInt(
     A=&GenerateIntFromCounter(SimpleCounter:NiagaraDataInterfaceSimpleCounter=DataInterface() {propertyValues {GpuSyncMode=None, InitialValue=0}})
     B=2
    )
   ) {isSetParameters=true, module=SetVariables_#5}
   /SetVariables(Particles.ShowBackFire:NiagaraInt32=/Niagara/DynamicInputs/Subtract/Subtract_Integer(A:NiagaraInt32=1, B:NiagaraInt32=$Particles.PlacementPicker)) {isSetParameters=true, module=SetVariables_#6}
   &InitializeParticle(
    Color=<rgba 1 1 1 0.05>
    Lifetime=&Lerp_Float(A=1, B=0.5, Alpha=&FN_FrontEndFloatModifier(Gameplay=$User.ForwardDot, Front End=0.02))
    Mesh Scale=&Multiply_Vector(
     A=&SelectVectorFromArray(
      Direct Array Index=$Particles.PlacementPicker
      Vector Selection Array=DataInterface() {propertyValues {FloatData [<0.18 0.18 0.06>, <0.18 0.18 0.09>]}}
     )
     B=&Lerp_Vector(
      A=<1 1 1>
      B=&RandomRangeVector(Minimum=<0.5 0.5 1>, Maximum=<1 1 1.5>)
      Alpha=&FN_FrontEndFloatModifier(Gameplay=$System.ThrusterPower, Front End=1)
     )
    )
    Module.Write Color=true
    Module.Write Lifetime=true
    Module.Write Mass=false
    Module.Write Material Random=false
    Module.Write Position=false
    Module.Write Scale=true
    Module.Write SpriteRotation=false
    Module.Write SpriteSize=false
    Position=&SimulationPosition()
    Sprite Size=<5 5>
   )
   &DynamicMaterialParameters(
    Index 0 Param 1=&RandomRangeFloat(Minimum=0, Maximum=1)
    Index 0 Param 4=&Lerp_Float(A=0, B=4, Alpha=&FN_FrontEndFloatModifier(Gameplay=$System.ThrusterPower, Front End=1))
    Module.Param0WriteEnabled=true
    Module.Param1WriteEnabled=false
    Module.Param2WriteEnabled=false
    Module.Param3WriteEnabled=true
   )
   &InitialMeshOrientation(
    Mesh Axis To Orient=<0 0 1>
    Mesh Orientation Mode=::NewEnumerator2
    Orientation Vector=&RotateVectorByQuaternion(Vector To Rotate=<1 0 0>, Quaternion=$Particles.SkeletalMeshLocation001.SampledRotation)
    Rotation=&RandomRangeVector(Minimum=<0 0 0>, Maximum=<0 0 1>)
    Use Rotation Vector=true
   )
  ]
  particleUpdateScript [
   &ParticleState()
   &SkeletalMeshLocation(
    Bone / Socket Index=&RandomRangeInt(Minimum=0, Maximum=1)
    MaskSpawnGroup=false
    Mesh Sampling Type=::NewEnumerator5
    Orientation Sampling=::NewEnumerator1
    Sampled Position Offset=<-2 0 0>
    Skeletal Mesh=$User.Skeletal Mesh
    Socket Sampling Mode=::NewEnumerator5
   )
   &ScaleMeshSize(Scale Factor=&VectorFromCurve(
     VectorCurve=DataInterface() {propertyValues {
       XCurve {keys [<at 0 1>,<at 1 1>]}
       YCurve {keys [<at 0 1>,<at 1 1>]}
       ZCurve {keys [<at 0.9 1 RCIM_Cubic>, {arriveTangent=-4.873, interpMode=RCIM_Cubic, leaveTangent=-4.873, tangentMode=RCTM_User, time=1, value=0}]}
      }}
     Scale Curve=&Lerp_Vector(A=<0.6 0.6 2>, B=<0.65 0.65 2.5>, Alpha=$System.ThrusterPower)
    ))
   &ScaleColor(
    Scale Alpha=&Lerp_Float(
     A=0
     B=&Lerp_Float(A=0.5, B=1, Alpha=&FN_FrontEndFloatModifier(Gameplay=$System.ThrusterPower, Front End=1))
     Alpha=$Particles.NormalizedAge
    )
    Scale RGB=&Lerp_Vector(
     A=<1 0.5 0.5>
     B=<1 0.5 0.5>
     Alpha=&FN_FrontEndFloatModifier(Gameplay=$System.ThrusterPower, Front End=1)
    )
    ScaleA=true
    ScaleRGB=true
   )
   /SetVariables(Particles.Color_Particle_A:LinearColor=<rgba 2 0.1333 0 1>, Particles.Color_Particle_B:LinearColor=<rgba 2 1.0667 0.6 1>) {isSetParameters=true}
  ]
  rendererClasses [&NiagaraMeshRendererProperties]
  renderers [
   {rendererClass=&NiagaraMeshRendererProperties, rendererIndex=0}
   {rendererClass=&NiagaraMeshRendererProperties, rendererIndex=1}
  ]
 }
 Flame_Mesh002 {
  emitterUpdateScript [
   &EmitterState(Life Cycle Mode=::NewEnumerator1, Loop Behavior=::NewEnumerator1, Loop Delay=0.4, Loop Duration=1.2, UseLoopDelay=true)
   &SpawnRate(
    SpawnRate=&Lerp_Float(
     A=80
     B=140
     Alpha=&FN_FrontEndFloatModifier(
      Gameplay=&FloatFromCurve(
       FloatCurve=DataInterface() {propertyValues {Curve {keys [{tangentMode=RCTM_User, time=0, value=0}, {tangentMode=RCTM_User, time=1, value=1}]}}}
       CurveIndex=$Emitter.Age
      )
      Front End=1
     )
    )
    Use Spawn Probability=false
   )
   &SpawnRate(Spawn Probability=$System.GenerateForwardBurst, SpawnRate=140, Use Spawn Probability=true) {enabled=false, module=SpawnRate001}
  ]
  particleSpawnScript [
   &SkeletalMeshLocation(
    Bone / Socket Index=&RandomRangeInt(Minimum=0, Maximum=1)
    MaskSpawnGroup=false
    Mesh Sampling Type=::NewEnumerator5
    Orientation Sampling=::NewEnumerator1
    Sampled Position Offset=<0 0 1.5>
    Skeletal Mesh=$User.Skeletal Mesh
    Socket Sampling Mode=::NewEnumerator5
   ) {module=SkeletalMeshLocation001}
   /SetVariables(
    Particles.PlacementPicker:NiagaraInt32=&Divide_Int(
     A:NiagaraInt32=&ModuloInt(
      A=&GenerateIntFromCounter(SimpleCounter:NiagaraDataInterfaceSimpleCounter=DataInterface() {propertyValues {GpuSyncMode=None, InitialValue=0}})
      B=4
     )
     B:NiagaraInt32=2
    )
    Particles.SpawnGroup=&ModuloInt(
     A=&GenerateIntFromCounter(SimpleCounter:NiagaraDataInterfaceSimpleCounter=DataInterface() {propertyValues {GpuSyncMode=None, InitialValue=0}})
     B=2
    )
   ) {isSetParameters=true, module=SetVariables_#5}
   /SetVariables(Particles.ShowBackFire:NiagaraInt32=&Subtract_Integer(A:NiagaraInt32=1, B:NiagaraInt32=$Particles.PlacementPicker)) {isSetParameters=true, module=SetVariables_#6}
   &InitializeParticle(
    Color=<rgba 1 1 1 0.05>
    Lifetime=&Lerp_Float(
     A=1
     B=0.5
     Alpha=&FN_FrontEndFloatModifier(
      Gameplay=&FloatFromCurve(
       FloatCurve=DataInterface() {propertyValues {Curve {keys [{tangentMode=RCTM_User, time=0, value=0}, {tangentMode=RCTM_User, time=1, value=1}]}}}
       CurveIndex=$Emitter.Age
      )
      Front End=0.02
     )
    )
    Mesh Scale=&Multiply_Vector(
     A=&SelectVectorFromArray(
      Direct Array Index=$Particles.PlacementPicker
      Vector Selection Array=DataInterface() {propertyValues {FloatData [<0.23 0.23 0.06>, <0.23 0.23 0.09>]}}
     )
     B=&Lerp_Vector(
      A=<1 1 1>
      B=&RandomRangeVector(Minimum=<0.5 0.5 1>, Maximum=<1 1 1.5>)
      Alpha=&FN_FrontEndFloatModifier(
       Gameplay=&FloatFromCurve(
        FloatCurve=DataInterface() {propertyValues {Curve {keys [{tangentMode=RCTM_User, time=0, value=0}, {tangentMode=RCTM_User, time=1, value=1}]}}}
        CurveIndex=$Emitter.Age
       )
       Front End=1
      )
     )
    )
    Module.Write Color=true
    Module.Write Lifetime=true
    Module.Write Mass=false
    Module.Write Material Random=false
    Module.Write Position=false
    Module.Write Scale=true
    Module.Write SpriteRotation=false
    Module.Write SpriteSize=false
    Position=&SimulationPosition()
    Sprite Size=<5 5>
   )
   &DynamicMaterialParameters(
    Index 0 Param 1=&RandomRangeFloat(Minimum=0, Maximum=1)
    Index 0 Param 4=&Lerp_Float(
     A=0
     B=4
     Alpha=&FN_FrontEndFloatModifier(
      Gameplay=&FloatFromCurve(
       FloatCurve=DataInterface() {propertyValues {Curve {keys [{tangentMode=RCTM_User, time=0, value=0}, {tangentMode=RCTM_User, time=1, value=1}]}}}
       CurveIndex=$Emitter.Age
      )
      Front End=1
     )
    )
    Module.Param0WriteEnabled=true
    Module.Param1WriteEnabled=false
    Module.Param2WriteEnabled=false
    Module.Param3WriteEnabled=true
   )
   &InitialMeshOrientation(
    Mesh Axis To Orient=<0 0 1>
    Mesh Orientation Mode=::NewEnumerator2
    Orientation Vector=&RotateVectorByQuaternion(Vector To Rotate=<1 0 0>, Quaternion=$Particles.SkeletalMeshLocation001.SampledRotation)
    Rotation=&RandomRangeVector(Minimum=<0 0 0>, Maximum=<0 0 1>)
    Use Rotation Vector=true
   )
  ]
  particleUpdateScript [
   &ParticleState()
   &SkeletalMeshLocation(
    Bone / Socket Index=&RandomRangeInt(Minimum=0, Maximum=1)
    MaskSpawnGroup=false
    Mesh Sampling Type=::NewEnumerator5
    Orientation Sampling=::NewEnumerator1
    Sampled Position Offset=<-5 0 0>
    Skeletal Mesh=$User.Skeletal Mesh
    Socket Sampling Mode=::NewEnumerator5
   )
   &ScaleMeshSize(Scale Factor=&VectorFromCurve(
     VectorCurve=DataInterface() {propertyValues {
       XCurve {keys [<at 0 1>,<at 1 1>]}
       YCurve {keys [<at 0 1>,<at 1 1>]}
       ZCurve {keys [<at 0.9 1 RCIM_Cubic>, {arriveTangent=-4.873, interpMode=RCIM_Cubic, leaveTangent=-4.873, tangentMode=RCTM_User, time=1, value=0}]}
      }}
     Scale Curve=&Lerp_Vector(
      A=<0.4 0.4 0.5>
      B=<0.4 0.4 0.5>
      Alpha=&FloatFromCurve(
       FloatCurve=DataInterface() {propertyValues {Curve {keys [{interpMode=RCIM_Cubic, tangentMode=RCTM_User, time=0, value=0}, {interpMode=RCIM_Cubic, tangentMode=RCTM_User, time=1, value=1}]}}}
       CurveIndex=$Emitter.Age
      )
     )
    ))
   &ScaleColor(
    Scale Alpha=&Lerp_Float(
     A=0
     B=&Lerp_Float(
      A=0
      B=1
      Alpha=&FN_FrontEndFloatModifier(
       Gameplay=&FloatFromCurve(
        FloatCurve=DataInterface() {propertyValues {Curve {keys [{tangentMode=RCTM_User, time=0, value=0}, {tangentMode=RCTM_User, time=1, value=1}]}}}
        CurveIndex=$Emitter.Age
       )
       Front End=1
      )
     )
     Alpha=$Particles.NormalizedAge
    )
    Scale RGB=&Lerp_Vector(
     A=<1 0.5 0.5>
     B=<1 0.5 0.5>
     Alpha=&FN_FrontEndFloatModifier(
      Gameplay=&FloatFromCurve(
       FloatCurve=DataInterface() {propertyValues {Curve {keys [{tangentMode=RCTM_User, time=0, value=0}, {tangentMode=RCTM_User, time=1, value=1}]}}}
       CurveIndex=$Emitter.Age
      )
      Front End=1
     )
    )
    ScaleA=true
    ScaleRGB=true
   )
  ]
  rendererClasses [&NiagaraMeshRendererProperties]
  renderers [{rendererClass=&NiagaraMeshRendererProperties, rendererIndex=0}]
 }
 Flame_Mesh003 {
  emitterUpdateScript [
   &EmitterState(Life Cycle Mode=::NewEnumerator1, Loop Behavior=::NewEnumerator1, Loop Delay=0.2, Loop Duration=1.2, UseLoopDelay=false)
   &SpawnRate(
    SpawnRate=&Lerp_Float(A=80, B=140, Alpha=&FN_FrontEndFloatModifier(Gameplay=$System.ThrusterPower, Front End=1))
    Use Spawn Probability=false
   ) {enabled=false}
   &SpawnRate(
    SpawnRate=&Lerp_Float(A=80, B=140, Alpha=&FN_FrontEndFloatModifier(Gameplay=1, Front End=1))
    Use Spawn Probability=false
   ) {module=SpawnRate002}
   &SpawnRate(SpawnRate=140, Use Spawn Probability=true) {module=SpawnRate001}
  ]
  particleSpawnScript [
   &SkeletalMeshLocation(
    Bone / Socket Index=&RandomRangeInt(Minimum=0, Maximum=1)
    MaskSpawnGroup=false
    Mesh Sampling Type=::NewEnumerator5
    Orientation Sampling=::NewEnumerator1
    Sampled Position Offset=<0 0 1.5>
    Skeletal Mesh=$User.Skeletal Mesh
    Socket Sampling Mode=::NewEnumerator5
   ) {module=SkeletalMeshLocation001}
   /SetVariables(
    Particles.PlacementPicker:NiagaraInt32=&Divide_Int(
     A:NiagaraInt32=&ModuloInt(
      A=&GenerateIntFromCounter(SimpleCounter:NiagaraDataInterfaceSimpleCounter=DataInterface() {propertyValues {GpuSyncMode=None, InitialValue=0}})
      B=4
     )
     B:NiagaraInt32=2
    )
    Particles.SpawnGroup=&ModuloInt(
     A=&GenerateIntFromCounter(SimpleCounter:NiagaraDataInterfaceSimpleCounter=DataInterface() {propertyValues {GpuSyncMode=None, InitialValue=0}})
     B=2
    )
   ) {isSetParameters=true, module=SetVariables_#5}
   /SetVariables(
    Particles.PlacementPicker:NiagaraInt32=&Divide_Int(
     A:NiagaraInt32=&ModuloInt(
      A=&GenerateIntFromCounter(SimpleCounter:NiagaraDataInterfaceSimpleCounter=DataInterface() {propertyValues {GpuSyncMode=None, InitialValue=0}})
      B=4
     )
     B:NiagaraInt32=2
    )
    Particles.SpawnGroup=&ModuloInt(
     A=&GenerateIntFromCounter(SimpleCounter:NiagaraDataInterfaceSimpleCounter=DataInterface() {propertyValues {GpuSyncMode=None, InitialValue=0}})
     B=2
    )
   ) {enabled=false, isSetParameters=true}
   /SetVariables(Particles.ShowBackFire:NiagaraInt32=&Subtract_Integer(A:NiagaraInt32=1, B:NiagaraInt32=$Particles.PlacementPicker)) {isSetParameters=true, module=SetVariables_#6}
   &InitializeParticle(
    Color=<rgba 1 1 1 0.02>
    Lifetime=&Lerp_Float(
     A=0.2
     B=0.5
     Alpha=&FN_FrontEndFloatModifier(
      Gameplay=&FloatFromCurve(FloatCurve=DataInterface() {propertyValues {Curve {keys [<at 0 1>,<at 1 0>]}}}, CurveIndex=$Emitter.Age)
      Front End=0.02
     )
    )
    Mesh Scale=&Multiply_Vector(
     A=&SelectVectorFromArray(
      Direct Array Index=$Particles.PlacementPicker
      Vector Selection Array=DataInterface() {propertyValues {FloatData [<0.15 0.15 0.03>, <0.15 0.15 0.06>]}}
     )
     B=&Lerp_Vector(
      A=<1 1 1>
      B=&RandomRangeVector(Minimum=<0.5 0.5 1>, Maximum=<1 1 1.5>)
      Alpha=&FN_FrontEndFloatModifier(Gameplay=$System.ThrusterPower, Front End=1)
     )
    )
    Module.Write Color=true
    Module.Write Lifetime=true
    Module.Write Mass=false
    Module.Write Material Random=false
    Module.Write Position=false
    Module.Write Scale=true
    Module.Write SpriteRotation=false
    Module.Write SpriteSize=false
    Position=&SimulationPosition()
    Sprite Size=<5 5>
   )
   &DynamicMaterialParameters(
    Index 0 Param 1=&RandomRangeFloat(Minimum=0, Maximum=1)
    Index 0 Param 4=&Lerp_Float(A=0, B=4, Alpha=&FN_FrontEndFloatModifier(Gameplay=$System.ThrusterPower, Front End=1))
    Module.Param0WriteEnabled=true
    Module.Param1WriteEnabled=false
    Module.Param2WriteEnabled=false
    Module.Param3WriteEnabled=true
   ) {enabled=false}
   &InitialMeshOrientation(
    Mesh Axis To Orient=<0 0 1>
    Mesh Orientation Mode=::NewEnumerator2
    Orientation Vector=&RotateVectorByQuaternion(Vector To Rotate=<1 0 0>, Quaternion=$Particles.SkeletalMeshLocation001.SampledRotation)
    Rotation=&RandomRangeVector(Minimum=<0 0 0>, Maximum=<0 0 1>)
    Use Rotation Vector=true
   )
  ]
  particleUpdateScript [
   &ParticleState()
   &SkeletalMeshLocation(
    Bone / Socket Index=&RandomRangeInt(Minimum=0, Maximum=1)
    MaskSpawnGroup=false
    Mesh Sampling Type=::NewEnumerator5
    Orientation Sampling=::NewEnumerator1
    Sampled Position Offset=<-10 0 0>
    Skeletal Mesh=$User.Skeletal Mesh
    Socket Sampling Mode=::NewEnumerator5
   )
   &ScaleMeshSize(Scale Factor=&VectorFromCurve(
     VectorCurve=DataInterface() {propertyValues {
       XCurve {keys [<at 0 1>,<at 1 1>]}
       YCurve {keys [<at 0 1>,<at 1 1>]}
       ZCurve {keys [<at 0.9 1 RCIM_Cubic>, {arriveTangent=-4.873, interpMode=RCIM_Cubic, leaveTangent=-4.873, tangentMode=RCTM_User, time=1, value=0}]}
      }}
     Scale Curve=&Lerp_Vector(
      A=<0.1 0.1 1>
      B=<0.7 0.7 1>
      Alpha=&FloatFromCurve(
       FloatCurve=DataInterface() {propertyValues {Curve {keys [
           {arriveTangent=1.9382, interpMode=RCIM_Cubic, time=0.1, value=0.5}
           {arriveTangent=0.125, interpMode=RCIM_Cubic, time=0.3, value=0.2}
           {arriveTangent=0.4852, interpMode=RCIM_Cubic, leaveTangent=0.4852, tangentMode=RCTM_User, time=1, value=1}
          ]}}}
       CurveIndex=$Emitter.Age
      )
     )
    ))
   &ScaleColor(
    Scale Alpha=&Lerp_Float(
     A=0
     B=&Lerp_Float(
      A=0
      B=1
      Alpha=&FN_FrontEndFloatModifier(
       Gameplay=&FloatFromCurve(
        FloatCurve=DataInterface() {propertyValues {Curve {keys [{tangentMode=RCTM_User, time=0, value=1}, {tangentMode=RCTM_User, time=1, value=0}]}}}
        CurveIndex=$Emitter.Age
       )
       Front End=1
      )
     )
     Alpha=$Particles.NormalizedAge
    )
    Scale RGB=&MakeVectorFromLinearColorRGB(Color=&ColorFromCurve(ColorCurve=DataInterface() {propertyValues {
        AlphaCurve {keys [<at 0 1 RCIM_Cubic>,<at 1 1 RCIM_Cubic>]}
        BlueCurve {keys [<at 0.2301 1 RCIM_Cubic>,<at 0.5656 0.1 RCIM_Cubic>]}
        GreenCurve {keys [<at 0.2301 0.2605 RCIM_Cubic>,<at 0.5656 0.0167 RCIM_Cubic>]}
        RedCurve {keys [<at 0.2301 0 RCIM_Cubic>,<at 0.5656 0 RCIM_Cubic>]}
       }}))
    ScaleA=true
    ScaleRGB=false
   )
   /SetVariables(Particles.ColorAlt:LinearColor=&MakeLinearColorFromVectorAndFloat(
     "Vector (RGB)"=&MakeVectorFromLinearColorRGB(Color=&ColorFromCurve(ColorCurve=DataInterface() {propertyValues {
         AlphaCurve {keys [<at 0 1 RCIM_Cubic>,<at 1 1 RCIM_Cubic>]}
         BlueCurve {keys [<at 0.2301 1 RCIM_Cubic>,<at 0.5656 0.1 RCIM_Cubic>]}
         GreenCurve {keys [<at 0.2301 0.2605 RCIM_Cubic>,<at 0.5656 0.0167 RCIM_Cubic>]}
         RedCurve {keys [<at 0.2301 0 RCIM_Cubic>,<at 0.5656 0 RCIM_Cubic>]}
        }}))
     "Float (Alpha)"=&Lerp_Float(
      A=0
      B=&Lerp_Float(
       A=0.1
       B=20
       Alpha=&FN_FrontEndFloatModifier(
        Gameplay=&FloatFromCurve(FloatCurve=DataInterface() {propertyValues {Curve {keys [<at 0 1>,<at 1 0>]}}}, CurveIndex=$Emitter.Age)
        Front End=0.02
       )
      )
      Alpha=$Particles.NormalizedAge
     )
    )) {isSetParameters=true}
   &DynamicMaterialParameters(
    Index 0 Param 1=1
    Index 0 Param 2=0.3
    Index 0 Param 4=&Lerp_Float(
     A=5
     B=20
     Alpha=&FN_FrontEndFloatModifier(
      Gameplay=&FloatFromCurve(FloatCurve=DataInterface() {propertyValues {Curve {keys [<at 0 1>,<at 1 0>]}}}, CurveIndex=$Emitter.Age)
      Front End=2
     )
    )
    Index 1 Param 1=20
    Module.Param0WriteEnabled=true
    Module.Param0WriteEnabled1=true
    Module.Param1WriteEnabled=true
    Module.Param1WriteEnabled1=true
    Module.Param2WriteEnabled=true
    Module.Param2WriteEnabled1=true
    Module.Param3WriteEnabled=true
    Module.Param3WriteEnabled1=true
    Write Parameter Index 1=true
   ) {module=DynamicMaterialParameters001}
  ]
  rendererClasses [&NiagaraMeshRendererProperties]
  renderers [
   {rendererClass=&NiagaraMeshRendererProperties, rendererIndex=0}
   {rendererClass=&NiagaraMeshRendererProperties, rendererIndex=1}
  ]
 }
 Flame_Mesh004 {
  emitterUpdateScript [
   &EmitterState(Life Cycle Mode=::NewEnumerator1, Loop Behavior=::NewEnumerator1, Loop Duration=2, UseLoopDelay=true)
   &SpawnRate(
    SpawnRate=&Lerp_Float(A=80, B=140, Alpha=&FN_FrontEndFloatModifier(Gameplay=$System.ThrusterPower, Front End=1))
    Use Spawn Probability=false
   )
   &SpawnRate(Spawn Probability=$System.GenerateForwardBurst, SpawnRate=140, Use Spawn Probability=true) {enabled=false, module=SpawnRate001}
  ]
  particleSpawnScript [
   &SkeletalMeshLocation(
    Bone / Socket Index=&RandomRangeInt(Minimum=0, Maximum=1)
    MaskSpawnGroup=false
    Mesh Sampling Type=::NewEnumerator5
    Orientation Sampling=::NewEnumerator1
    Sampled Position Offset=<0 0 1.5>
    Skeletal Mesh=$User.Skeletal Mesh
    Socket Sampling Mode=::NewEnumerator5
   ) {module=SkeletalMeshLocation001}
   /SetVariables(
    Particles.PlacementPicker:NiagaraInt32=&Divide_Int(
     A:NiagaraInt32=&ModuloInt(
      A=&GenerateIntFromCounter(SimpleCounter:NiagaraDataInterfaceSimpleCounter=DataInterface() {propertyValues {GpuSyncMode=None, InitialValue=0}})
      B=4
     )
     B:NiagaraInt32=2
    )
    Particles.SpawnGroup=&ModuloInt(
     A=&GenerateIntFromCounter(SimpleCounter:NiagaraDataInterfaceSimpleCounter=DataInterface() {propertyValues {GpuSyncMode=None, InitialValue=0}})
     B=2
    )
   ) {isSetParameters=true, module=SetVariables_#5}
   /SetVariables(Particles.ShowBackFire:NiagaraInt32=&Subtract_Integer(A:NiagaraInt32=1, B:NiagaraInt32=$Particles.PlacementPicker)) {isSetParameters=true, module=SetVariables_#6}
   &InitializeParticle(
    Lifetime=&Lerp_Float(A=1, B=1, Alpha=&FN_FrontEndFloatModifier(Gameplay=$User.ForwardDot, Front End=0.02))
    Mesh Scale=&Multiply_Vector(
     A=&SelectVectorFromArray(
      Direct Array Index=$Particles.PlacementPicker
      Vector Selection Array=DataInterface() {propertyValues {FloatData [<0.23 0.23 0.06>, <0.23 0.23 0.09>]}}
     )
     B=&Lerp_Vector(
      A=<1 1 1>
      B=&RandomRangeVector(Minimum=<0.5 0.5 1>, Maximum=<1 1 1.5>)
      Alpha=&FN_FrontEndFloatModifier(Gameplay=$System.ThrusterPower, Front End=0)
     )
    )
    Module.Write Color=true
    Module.Write Lifetime=true
    Module.Write Mass=false
    Module.Write Material Random=false
    Module.Write Position=false
    Module.Write Scale=true
    Module.Write SpriteRotation=false
    Module.Write SpriteSize=false
    Position=&SimulationPosition()
    Sprite Size=<5 5>
   )
   &DynamicMaterialParameters(
    Index 0 Param 1=&RandomRangeFloat(Minimum=0, Maximum=1)
    Index 0 Param 4=&Lerp_Float(A=0, B=4, Alpha=&FN_FrontEndFloatModifier(Gameplay=$System.ThrusterPower, Front End=1))
    Module.Param0WriteEnabled=true
    Module.Param1WriteEnabled=false
    Module.Param2WriteEnabled=false
    Module.Param3WriteEnabled=true
   )
   &InitialMeshOrientation(
    Mesh Axis To Orient=<0 0 1>
    Mesh Orientation Mode=::NewEnumerator2
    Orientation Vector=&RotateVectorByQuaternion(Vector To Rotate=<1 0 0>, Quaternion=$Particles.SkeletalMeshLocation001.SampledRotation)
    Rotation=&RandomRangeVector(Minimum=<0 0 0>, Maximum=<0 0 1>)
    Use Rotation Vector=true
   )
  ]
  particleUpdateScript [
   &ParticleState()
   &SkeletalMeshLocation(
    Bone / Socket Index=&RandomRangeInt(Minimum=0, Maximum=1)
    MaskSpawnGroup=false
    Mesh Sampling Type=::NewEnumerator5
    Orientation Sampling=::NewEnumerator1
    Sampled Position Offset=<-5 0 0>
    Skeletal Mesh=$User.Skeletal Mesh
    Socket Sampling Mode=::NewEnumerator5
   )
   &ScaleMeshSize(Scale Factor=&VectorFromCurve(
     VectorCurve=DataInterface() {propertyValues {
       XCurve {keys [<at 0 1>,<at 1 1>]}
       YCurve {keys [<at 0 1>,<at 1 1>]}
       ZCurve {keys [<at 0.9 1 RCIM_Cubic>, {arriveTangent=-4.873, interpMode=RCIM_Cubic, leaveTangent=-4.873, tangentMode=RCTM_User, time=1, value=0}]}
      }}
     Scale Curve=&Lerp_Vector(
      A=<0.5 0.5 40>
      B=<0.5 0.5 2>
      Alpha=&FloatFromCurve(
       FloatCurve=DataInterface() {propertyValues {Curve {keys [{tangentMode=RCTM_User, time=0, value=0}, <at 0.5 0.8>, <at 1.6 0.95>, {tangentMode=RCTM_User, time=2, value=1}]}}}
       CurveIndex=$Emitter.Age
      )
     )
    ))
   &ScaleColor(
    Scale Alpha=&Lerp_Float(
     A=0
     B=&Lerp_Float(A=0, B=1, Alpha=&FN_FrontEndFloatModifier(Gameplay=$System.ThrusterPower, Front End=0.02))
     Alpha=$Particles.NormalizedAge
    )
    Scale RGB=&Lerp_Vector(
     A=<1 0.5 0.5>
     B=<1 0.5 0.5>
     Alpha=&FN_FrontEndFloatModifier(Gameplay=$System.ThrusterPower, Front End=1)
    )
    ScaleA=true
    ScaleRGB=true
   )
  ]
  rendererClasses [&NiagaraMeshRendererProperties]
  renderers [{rendererClass=&NiagaraMeshRendererProperties, rendererIndex=0}]
 }
 Flame_MeshCones {
  bEnabled=false
  emitterUpdateScript [
   &EmitterState()
   &SpawnRate(SpawnRate=16, Use Spawn Probability=false)
   &SpawnRate(Spawn Probability=$System.GenerateForwardBurst, SpawnRate=140, Use Spawn Probability=true) {enabled=false, module=SpawnRate001}
  ]
  particleSpawnScript [
   /Niagara/Modules/Spawn/Initialization/V2/InitializeParticle(
    Color Mode=::NewEnumerator0
    Lifetime Max=1
    Lifetime Min=0.5
    Lifetime Mode=::NewEnumerator1
    Mesh Scale=<0.12 0.12 0.1>
    Mesh Scale Mode=::NewEnumerator4
    UsePositionOffset=false
    Write Material Random=true
   )
   &SkeletalMeshLocation(
    Bone / Socket Index=&RandomRangeInt(Minimum=0, Maximum=1)
    MaskSpawnGroup=false
    Mesh Sampling Type=::NewEnumerator5
    Orientation Sampling=::NewEnumerator1
    Skeletal Mesh=$User.Skeletal Mesh
    Socket Sampling Mode=::NewEnumerator5
   )
   &InitialMeshOrientation(
    Mesh Axis To Orient=<0 0 1>
    Mesh Orientation Mode=::NewEnumerator2
    Orientation Vector=&RotateVectorByQuaternion(Vector To Rotate=<1 0 0>, Quaternion=$Particles.SkeletalMeshLocation001.SampledRotation)
    Use Rotation Vector=false
   )
   &DynamicMaterialParameters(
    Index 0 Param 1=&RandomRangeFloat(Minimum=0, Maximum=1)
    Index 0 Param 4=&Lerp_Float(A=2, B=3, Alpha=&FN_FrontEndFloatModifier(Gameplay=$User.ForwardDot, Front End=1))
    Module.Param0WriteEnabled=true
    Module.Param1WriteEnabled=false
    Module.Param2WriteEnabled=false
    Module.Param3WriteEnabled=true
   )
  ]
  particleUpdateScript [
   &ParticleState()
   &ScaleMeshSize(Scale Factor=&VectorFromCurve(
     VectorCurve=DataInterface() {propertyValues {
       XCurve {keys [<at 0 1>,<at 1 1>]}
       YCurve {keys [<at 0 1>,<at 1 1>]}
       ZCurve {keys [<at 0.9 1 RCIM_Cubic>, {arriveTangent=-4.873, interpMode=RCIM_Cubic, leaveTangent=-4.873, tangentMode=RCTM_User, time=1, value=0}]}
      }}
     Scale Curve=&Lerp_Vector(
      A=<1 1 1>
      B=<1 1 2>
      Alpha=&self:InterpolateFloat(Float to Interpolate:NiagaraFloat=$System.GenerateForwardBurst, Initial Value:NiagaraFloat=1, Rate Of Change:NiagaraFloat=20)
     )
    ))
  ]
  rendererClasses [&NiagaraMeshRendererProperties]
  renderers [{rendererClass=&NiagaraMeshRendererProperties, rendererIndex=0}]
 }
 Flame_MeshOval {
  bEnabled=false
  emitterUpdateScript [
   &EmitterState()
   &SpawnRate(SpawnRate=16, Use Spawn Probability=false) {enabled=false}
   &SpawnRate(SpawnRate=200, Use Spawn Probability=true) {module=SpawnRate001}
  ]
  particleSpawnScript [
   &InitializeParticle(
    Color=&MakeLinearColorFromVectorAndFloat("Vector (RGB)"=<1 1 1>, "Float (Alpha)"=0.02)
    Lifetime Max=1
    Lifetime Min=0.5
    Lifetime Mode=::NewEnumerator1
    Mesh Scale Max=<0.1 0.09 0.08>
    Mesh Scale Min=<0.075 0.085 0.04>
    Mesh Scale Mode=::NewEnumerator2
    UsePositionOffset=false
    Write Material Random=true
   )
   &SkeletalMeshLocation(
    Bone / Socket Index=2
    MaskSpawnGroup=false
    Mesh Sampling Type=::NewEnumerator5
    Sampled Position Offset=<0 0 1.5>
    Skeletal Mesh=$User.Skeletal Mesh
    Socket Sampling Mode=::NewEnumerator5
   ) {enabled=false}
   &InitialMeshOrientation(Mesh Axis To Orient=<0 0 1>, Mesh Orientation Mode=::NewEnumerator2, Use Rotation Vector=false)
   &DynamicMaterialParameters(
    Index 0 Param 1=&RandomRangeFloat(Minimum=0, Maximum=3)
    Index 0 Param 4=&Lerp_Float(A=1, B=3, Alpha=&FN_FrontEndFloatModifier(Gameplay=$System.ThrusterPower, Front End=1))
    Module.Param0WriteEnabled=true
    Module.Param1WriteEnabled=false
    Module.Param2WriteEnabled=false
    Module.Param3WriteEnabled=true
   )
  ]
  particleUpdateScript [
   &ParticleState()
   &SkeletalMeshLocation(
    Bone / Socket Index=2
    MaskSpawnGroup=false
    Mesh Sampling Type=::NewEnumerator5
    Skeletal Mesh=$User.Skeletal Mesh
    Socket Sampling Mode=::NewEnumerator5
   ) {module=SkeletalMeshLocation001}
   &ScaleMeshSize(Scale Factor=&VectorFromCurve(
     VectorCurve=DataInterface() {propertyValues {
       XCurve {keys [<at 0 1>,<at 1 1>]}
       YCurve {keys [<at 0 1>,<at 1 1>]}
       ZCurve {keys [<at 0.9 1 RCIM_Cubic>, {arriveTangent=-4.873, interpMode=RCIM_Cubic, leaveTangent=-4.873, tangentMode=RCTM_User, time=1, value=0}]}
      }}
     Scale Curve=&Lerp_Vector(A=<1 1 1>, B=<1 1 20>, Alpha=$System.ThrusterPowerCurve)
    )) {enabled=false}
   &ScaleMeshSize(Scale Factor=&VectorFromCurve(
     VectorCurve=DataInterface() {propertyValues {
       XCurve {keys [<at 0 1>,<at 1 1>]}
       YCurve {keys [<at 0 1>,<at 1 1>]}
       ZCurve {keys [<at 0.9 1 RCIM_Cubic>, {arriveTangent=-4.873, interpMode=RCIM_Cubic, leaveTangent=-4.873, tangentMode=RCTM_User, time=1, value=0}]}
      }}
     Scale Curve=&Lerp_Vector(A=<1 1 2>, B=<1 1 8>, Alpha=$System.ThrusterPower)
    )) {module=ScaleMeshSize001}
   &ScaleColor(
    Scale Alpha=&Lerp_Float(
     A=1
     B=&Lerp_Float(A=0, B=0, Alpha=&FN_FrontEndFloatModifier(Gameplay=$System.ThrusterPower, Front End=1))
     Alpha=$Particles.NormalizedAge
    )
    Scale RGB=&Lerp_Vector(
     A=<1 0.5 0.5>
     B=<1 0.5 0.5>
     Alpha=&FN_FrontEndFloatModifier(Gameplay=$System.ThrusterPower, Front End=1)
    )
    ScaleA=true
    ScaleRGB=true
   )
   &ScaleColor(
    Scale RGB=&MakeCustomVectorFromBool(Bool=$System.Timeline.IsRewinding, True Vector=<0 0 0>, False Vector=<0 0 0>)
    ScaleA=true
    ScaleRGB=true
   ) {enabled=false, module=ScaleColor001}
  ]
  rendererClasses [&NiagaraMeshRendererProperties]
  renderers [{rendererClass=&NiagaraMeshRendererProperties, rendererIndex=0}]
 }
 Light {
  emitterUpdateScript [
   &EmitterState(Life Cycle Mode=::NewEnumerator1, Loop Behavior=::NewEnumerator1, Loop Delay=0.2, Loop Duration Mode=::NewEnumerator1, UseLoopDelay=false)
   &SpawnBurst_Instantaneous(
    Spawn Count=2
    Spawn Probability=&MakeCustomFloatFromBool(
     Bool=&SetBoolByIntComparison(A=$Emitter.LoopCount, B=0, Comparison Type=::NewEnumerator4)
     True Float=1
     False Float=0
    )
    Use Spawn Probability=true
   )
   &SpawnRate(SpawnRate=3, Use Spawn Probability=false) {enabled=false}
  ]
  particleSpawnScript [
   /Niagara/Modules/Spawn/Initialization/InitializeParticle(
    Color=<rgba 20 6.5 2 1>
    Module.Write Color=true
    Module.Write Lifetime=true
    Module.Write Mass=true
    Module.Write Material Random=true
    Module.Write Position=false
    Module.Write Scale=false
    Module.Write SpriteRotation=false
    Module.Write SpriteSize=true
    Position=&SimulationPosition()
    Sprite Size=<4 4>
   )
   &FN_SetSpawnGroup(Number of Spawn Groups=2)
   &CameraOffset(Camera Offset Amount=3) {enabled=false}
  ]
  particleUpdateScript [
   &ParticleState(
    Kill Particles When Lifetime Has Elapsed=false
    Let Infinitely Lived Particles Die When Emitter Deactivates=true
    Loop Particles Lifetime=true
   )
   &SkeletalMeshLocation(
    Bone / Socket Index=$Particles.SpawnGroup
    Inherit Velocity=false
    MaskSpawnGroup=false
    Mesh Sampling Type=::NewEnumerator5
    Sampled Position Offset=&Add_Vector(A=$System.OffsetLoopFromEngineBone, B=<5 0 0>)
    Sampled Position Offset Coordinate Space:ENiagaraMeshTransforms=::NewEnumerator3
    Skeletal Mesh=$User.Skeletal Mesh
    Socket Sampling Mode=::NewEnumerator5
   )
   &FN_Light_v3(
    Color=$System.SysColorA
    Create Color Attribute=true
    Create Exponent Attribute=true
    Create Radius Attribute=true
    Enable Particle Light Enable=true
    Exponent=1
    Light Enabled=true
    Radius:NiagaraFloat=&Lerp_Float(
     A=0
     B=&RandomRangeFloat(Minimum=25, Maximum=45, Evaluation Type=::NewEnumerator1)
     Alpha=$System.ThrusterPower
    )
    Scale Color:NiagaraFloat=&FloatFromCurve(
     FloatCurve=DataInterface() {propertyValues {Curve {keys [<at 0 0.5 RCIM_Cubic>,<at 0.5 1 RCIM_Cubic>,<at 1 0.5>]}}}
     CurveIndex=$Emitter.NormalizedLoopAge
     Scale Curve=&RandomRangeFloat(
      Minimum=50
      Maximum=200
      Evaluation Type=::NewEnumerator1
      Randomness Mode=::NewEnumerator1
      Random Seed=&Add_Integer(A=$Particles.SpawnGroup, B=&Add_Integer(A=$Emitter.LoopCount, B=$System.LoopCount))
      Fixed Random Seed=true
     )
    )
   )
   &ScaleColor(
    Scale Alpha=&FloatFromCurve(
     FloatCurve=DataInterface() {propertyValues {Curve {keys [
         {interpMode=RCIM_Cubic, tangentMode=RCTM_User, time=0, value=0}
         {interpMode=RCIM_Cubic, tangentMode=RCTM_User, time=0.5, value=1}
         {tangentMode=RCTM_User, time=1, value=0}
        ]}}}
     CurveIndex=$System.ThrusterPower
    )
    ScaleA=true
    ScaleRGB=true
   )
  ]
  rendererClasses [&NiagaraSpriteRendererProperties, &NiagaraLightRendererProperties]
  renderers [
   {rendererClass=&NiagaraSpriteRendererProperties}
   {rendererClass=&NiagaraLightRendererProperties, rendererIndex=1}
  ]
 }
 Light001 {
  emitterUpdateScript [
   &EmitterState()
   &SpawnBurst_Instantaneous(Spawn Count=4, Use Spawn Probability=false) {enabled=false}
   &SpawnRate(SpawnRate=200, Use Spawn Probability=false) {enabled=false}
   &SpawnBurst_Instantaneous(Spawn Count=2, Use Spawn Probability=false, UseLoopCountLimit=false) {module=SpawnBurst_Instantaneous001}
  ]
  particleSpawnScript [
   &InitializeParticle(
    Color=<rgba 0 3.3333 20 1>
    Lifetime=5
    Module.Write Color=true
    Module.Write Lifetime=true
    Module.Write Mass=true
    Module.Write Material Random=true
    Module.Write Position=false
    Module.Write Scale=false
    Module.Write SpriteRotation=false
    Module.Write SpriteSize=true
    Position=&SimulationPosition()
    Sprite Size=&Multiply_Vector2D_ByFloat(
     Float=&FloatFromCurve(
      FloatCurve=DataInterface() {propertyValues {Curve {keys [{tangentMode=RCTM_User, time=0, value=1}, {tangentMode=RCTM_User, time=1, value=0}]}}}
      CurveIndex=$Emitter.Age
     )
     Vector2D=<200 200>
    )
   )
   &FN_SetSpawnGroup(Number of Spawn Groups=2)
   /SetVariables(Particles.LightRadius=200) {isSetParameters=true}
   /SetVariables(Particles.FN_Light_v3.Color:LinearColor=<rgba 20 0 6.6817 1>) {isSetParameters=true}
   &CameraOffset(Camera Offset Amount=3) {enabled=false}
  ]
  particleUpdateScript [
   &ParticleState()
   &SkeletalMeshLocation(
    Bone / Socket Index=$Particles.SpawnGroup
    Inherit Velocity=false
    MaskSpawnGroup=false
    Mesh Sampling Type=::NewEnumerator5
    Sampled Position Offset=&Add_Vector(A=$System.OffsetLoopFromEngineBone, B=<-1 0 -4>)
    Sampled Position Offset Coordinate Space:ENiagaraMeshTransforms=::NewEnumerator3
    Skeletal Mesh=$User.Skeletal Mesh
    Socket Sampling Mode=::NewEnumerator5
   )
   &FN_Light_v3(
    Color=<rgba 0 3.3333 20 1>
    Create Color Attribute=true
    Create Exponent Attribute=true
    Create Radius Attribute=true
    Enable Particle Light Enable=true
    Exponent=1
    Light Enabled=true
    Radius=1
    Scale Color:NiagaraFloat=&FloatFromCurve(
     FloatCurve=DataInterface() {propertyValues {Curve {keys [
         <at 0 1 RCIM_Cubic>
         {arriveTangent=-0.3333, interpMode=RCIM_Cubic, leaveTangent=-0.3333, time=0.5, value=1}
         {arriveTangent=-0.4, interpMode=RCIM_Cubic, leaveTangent=-0.4, time=1.5, value=1}
         <at 3 0.5>
        ]}}}
     CurveIndex=$Emitter.Age
     Scale Curve=20
    )
   )
   /SetVariables(Particles.LightRadius=50) {enabled=false, isSetParameters=true}
   /SetVariables(Particles.LightRadius=&FloatFromCurve(
     FloatCurve=DataInterface() {propertyValues {Curve {keys [
         {interpMode=RCIM_Cubic, tangentMode=RCTM_User, time=0, value=1}
         {interpMode=RCIM_Cubic, tangentMode=RCTM_User, time=0.2, value=2}
         <at 1 2 RCIM_Cubic>
         <at 3 0.2>
        ]}}}
     CurveIndex=&Multiply_Float(A=$Emitter.Age, B=0.5)
     Scale Curve=25
    )) {isSetParameters=true}
   &ScaleColor(
    Scale Alpha=&Multiply_Float(
     A=&FloatFromCurve(
      FloatCurve=DataInterface() {propertyValues {Curve {keys [{tangentMode=RCTM_User, time=0, value=-0.5}, {tangentMode=RCTM_User, time=1, value=1}]}}}
      CurveIndex=&FloatFromCurve(
       FloatCurve=DataInterface() {propertyValues {Curve {keys [{tangentMode=RCTM_User, time=0, value=0}, {tangentMode=RCTM_User, time=1, value=1}]}}}
       CurveIndex=$Emitter.Age
      )
     )
     B=&FloatFromCurve(
      FloatCurve=DataInterface() {propertyValues {Curve {keys [
          {interpMode=RCIM_Cubic, tangentMode=RCTM_User, time=0, value=0.5}
          {interpMode=RCIM_Cubic, tangentMode=RCTM_User, time=0.5, value=1}
          {tangentMode=RCTM_User, time=1, value=0.5}
         ]}}}
      CurveIndex=$System.LoopedAge
      Scale Curve=0.5
     )
    )
    ScaleA=true
    ScaleRGB=true
   ) {enabled=false}
  ]
  rendererClasses [&NiagaraSpriteRendererProperties, &NiagaraLightRendererProperties]
  renderers [
   {rendererClass=&NiagaraSpriteRendererProperties}
   {rendererClass=&NiagaraLightRendererProperties, rendererIndex=1}
  ]
 }
 Lightning {
  bEnabled=false
  emitterUpdateScript [
   &EmitterState()
   &SpawnRate(Spawn Probability=0.2, SpawnRate=&UniformRangedFloat(Minimum=5, Maximum=20), Use Spawn Probability=true)
   &SpawnBurst_Instantaneous(
    Spawn Count=&RandomRangeInt(Minimum=0, Maximum=3)
    Spawn Probability=0.25
    Spawn Time=&RandomRangeFloat(Minimum=0, Maximum=1)
    Use Spawn Probability=true
   ) {enabled=false}
  ]
  particleSpawnScript [
   &InitializeParticle(
    Color=&MultiplyLinearColor(
     Linear Color=<rgba 25 2.8333 1.25 1>
     Scale Factor=&MakeVector4_FromFloat(Float=&RandomRangeFloat(Minimum=0.04, Maximum=1))
    )
    Lifetime=&Multiply_Float(
     A=&FN_FrontEndFloatModifier(Gameplay=$User.ForwardDot, Front End=1)
     B=&RandomRangeFloat(Minimum=0.1, Maximum=0.2)
    )
    Mesh Scale=&UniformRangedVector(Minimum=<0.05 0.01 0.01>, Maximum=<0.125 0.03 0.03>)
    Module.Write Color=true
    Module.Write Lifetime=true
    Module.Write Mass=true
    Module.Write Material Random=true
    Module.Write Position=true
    Module.Write Scale=true
    Module.Write SpriteRotation=false
    Module.Write SpriteSize=false
    Position=&SimulationPosition()
    Sprite Size=<25 25>
   )
   /Niagara/Modules/Spawn/Orientation/V2/InitialMeshOrientation(
    Orientation Vector=<0 1 0>
    Rotation=&UniformRangedVector(Minimum=<-360 0 0>, Maximum=<360 0 0>)
    Use Orientation Vector=true
    Use Rotation Vector=true
   )
   &DynamicMaterialParameters(
    Index 0 Param 1=&RandomRangeFloat(Minimum=0.25, Maximum=1.5)
    Index 0 Param 2=&RandomRangeFloat(Minimum=0.5, Maximum=1.5)
    Index 0 Param 3=1
    Module.Param0WriteEnabled=true
    Module.Param1WriteEnabled=true
    Module.Param2WriteEnabled=true
    Module.Param3WriteEnabled=true
   )
   &SkeletalMeshLocation(
    Inherit Velocity=false
    MaskSpawnGroup=false
    Mesh Sampling Type=::NewEnumerator5
    Module.Override Randomness=false
    Sampled Position Offset=&Add_Vector(A=$System.OffsetLoopFromEngineBone, B=&RandomRangeVector(Minimum=<-3 0 -6>, Maximum=<3 0 6>))
    Sampled Position Offset Coordinate Space:ENiagaraMeshTransforms=::NewEnumerator3
    Skeletal Mesh=$User.Skeletal Mesh
   )
  ]
  particleUpdateScript [
   &ParticleState()
   &SkeletalMeshLocation(
    Inherit Velocity=false
    Interpolate Mesh Sampling=false
    MaskSpawnGroup=false
    Mesh Sampling Type=::NewEnumerator5
    Module.Override Randomness=false
    Sampled Position Offset=&Add_Vector(A=$System.OffsetLoopFromEngineBone, B=&RandomRangeVector(Minimum=<-3 0 -6>, Maximum=<3 0 6>))
    Sampled Position Offset Coordinate Space:ENiagaraMeshTransforms=::NewEnumerator3
    Skeletal Mesh=$User.Skeletal Mesh
   ) {module=SkeletalMeshLocation001}
   &ScaleMeshSize(Scale Factor=&VectorFromCurve(VectorCurve=DataInterface() {propertyValues {
       XCurve {keys [
         {arriveTangent=2.6646, interpMode=RCIM_Cubic, leaveTangent=2.6646, tangentMode=RCTM_User, time=-0.0011, value=0.0216}
         {arriveTangent=0.0786, interpMode=RCIM_Cubic, leaveTangent=0.0786, time=0.5, value=0.95}
         {arriveTangent=-3.0502, interpMode=RCIM_Cubic, leaveTangent=-3.0502, tangentMode=RCTM_User, time=0.9954, value=0.1}
        ]}
       YCurve {keys [<at 0 0.3>,<at 1 1>]}
       ZCurve {keys [<at 0 0.3>,<at 1 1>]}
      }}))
   &SolveForcesAndVelocity(Clamp Velocity=false, Limit Acceleration=false)
  ]
  rendererClasses [&NiagaraMeshRendererProperties]
  renderers [{rendererClass=&NiagaraMeshRendererProperties, rendererIndex=0}]
 }
 MeshTrails {
  bEnabled=false
  emitterUpdateScript [
   &EmitterState()
   &SpawnBurst_Instantaneous(Spawn Count=HlslExpression() {hlslExpression="Emitter.LoopCount == 0"}, Use Spawn Probability=false) {module=SpawnBurst_Instantaneous001}
  ]
  particleSpawnScript [
   &InitializeParticle(
    Color=<rgba 1 1 1 0.125>
    Module.Write Color=true
    Module.Write Lifetime=true
    Module.Write Mass=false
    Module.Write Material Random=true
    Module.Write Position=false
    Module.Write Scale=false
    Module.Write SpriteRotation=false
    Module.Write SpriteSize=false
   )
   &self:Mesh_Trail_Setup(Trail Length=300, Trail Width=75) {module=Mesh_Trail_Setup}
  ]
  particleUpdateScript [
   &ParticleState(Kill Particles When Lifetime Has Elapsed=false, Let Infinitely Lived Particles Die When Emitter Deactivates=true)
   &SkeletalMeshLocation(
    Apply Sampled Orientation to Mesh Particles=true
    Apply To Particle Position=false
    Bone / Socket Index=2
    Bone Sampling Mode=::NewEnumerator4
    Inherit Velocity=false
    Interpolate Mesh Sampling=false
    MaskSpawnGroup=false
    Sampled Position Offset Coordinate Space:ENiagaraMeshTransforms=::NewEnumerator3
    Skeletal Mesh=$User.Skeletal Mesh
    Write Sampled Mesh Attrributes to Particle Attributes=true
   ) {module=SkeletalMeshLocation001}
   &SkeletalMeshLocation(
    Apply Sampled Orientation to Mesh Particles=true
    Apply To Particle Position=false
    Bone / Socket Index=3
    Bone Sampling Mode=::NewEnumerator4
    Inherit Velocity=false
    Interpolate Mesh Sampling=false
    MaskSpawnGroup=false
    Sampled Position Offset Coordinate Space:ENiagaraMeshTransforms=::NewEnumerator3
    Skeletal Mesh=$User.Skeletal Mesh
    Write Sampled Mesh Attrributes to Particle Attributes=true
   ) {module=SkeletalMeshLocation002}
   /SetVariables(
    Particles.TrailOrientation:Quat4f=&MultiplyQuaternion(Quaternion A=$Particles.SkeletalMeshLocation002.SampledRotation, Quaternion B=<0.23 0 1.453 1>)
    Particles.TrailPos:Vector3f=&Add_Vector(
     A=$Particles.SkeletalMeshLocation002.SampledPosition
     B=&RotateVectorByQuaternion(
      Vector To Rotate=&FN_FrontEndVector3Modifier(Gameplay=<94 17 -1>, Front End=<81 10 -1>)
      Quaternion=$Particles.SkeletalMeshLocation002.SampledRotation
     )
    )
   ) {isSetParameters=true}
   /SetVariables(
    Particles.TrailTwinOrientation:Quat4f=&MultiplyQuaternion(Quaternion A=$Particles.SkeletalMeshLocation001.SampledRotation, Quaternion B=<1 0.74 0.15 0>)
    Particles.TrailTwinPos:Vector3f=&Add_Vector(
     A=$Particles.SkeletalMeshLocation001.SampledPosition
     B=&RotateVectorByQuaternion(
      Vector To Rotate=&FN_FrontEndVector3Modifier(Gameplay=<-92 17 -1>, Front End=<-81 10 -1>)
      Quaternion=$Particles.SkeletalMeshLocation001.SampledRotation
     )
    )
   ) {isSetParameters=true}
   &self:Exhaust_to_DynaMat_02(
    ExhaustVector:Vector3f=&FN_FrontEndVector3Modifier(
     Gameplay=&Multiply_Vector(A=$System.CurrentExhaust, B=<1 5 1>)
     Front End=&self:Front_End_Exhaust_03(
      Rate Of Change:NiagaraFloat=10
      Velocity Input:Vector3f=&TransformVector(Vector=$Particles.SkeletalMeshLocation001.SampledVelocity, Source Space=::World, Destination Space=::Local)
      Velocity Multiplier:Vector3f=<0 30 0>
     )
    )
    SpeedMult=1
   ) {module=Exhaust_to_DynaMat_02}
   &self:Mesh_Trail_Length_Update_02(Dynamic Length Scale:NiagaraFloat=&FN_FrontEndFloatModifier(Gameplay=$System.TrailLength, Front End=1)) {module=Mesh_Trail_Length_Update_02}
   &ScaleColor(
    Scale Alpha=&FN_FrontEndFloatModifier(Gameplay=$System.TrailOpacity, Front End=1)
    ScaleA=true
    ScaleRGB=false
   )
   &Color(Color=&ColorFromCurve(ColorCurve=DataInterface() {propertyValues {
       AlphaCurve {keys [<at 0 20 RCIM_Cubic>]}
       BlueCurve {keys [<at 0 0 RCIM_Cubic>,<at 0.5257 0>,<at 1 5 RCIM_Cubic>]}
       GreenCurve {keys [<at 0 0 RCIM_Cubic>,<at 0.5257 5>,<at 1 0 RCIM_Cubic>]}
       RedCurve {keys [<at 0 5 RCIM_Cubic>,<at 0.5257 0>,<at 1 0 RCIM_Cubic>]}
      }})) {enabled=false}
  ]
  rendererClasses [&NiagaraMeshRendererProperties]
  renderers [
   {rendererClass=&NiagaraMeshRendererProperties, rendererIndex=0}
   {rendererClass=&NiagaraMeshRendererProperties, rendererIndex=1}
  ]
 }
 Sparks {
  bEnabled=false
  simTarget=GPUComputeSim
  emitterUpdateScript [
   &EmitterState()
   &SpawnRate(Spawn Probability=0.5, SpawnRate=60, Use Spawn Probability=true)
   &SpawnRate(SpawnGroup=1, SpawnRate=&Multiply_Float(A=500, B=$User.ForwardBurst), Use Spawn Probability=false) {module=SpawnRate001}
  ]
  particleSpawnScript [
   &InitializeParticle(
    Color=<rgba 10 5.25 2.5 1>
    Lifetime=&Multiply_Float(
     A=&FN_FrontEndFloatModifier(Gameplay=$User.ForwardDot, Front End=1)
     B=&RandomRangeFloat(Minimum=0.2, Maximum=0.4)
    )
    Module.Write Color=true
    Module.Write Lifetime=true
    Module.Write Mass=true
    Module.Write Material Random=true
    Module.Write Position=false
    Module.Write Scale=false
    Module.Write SpriteRotation=false
    Module.Write SpriteSize=true
    Position=&SimulationPosition()
    Sprite Size=&UniformRangedVector2D(Minimum=<1 0.25>, Maximum=<1.75 2>)
   )
   &SkeletalMeshLocation(
    Bone / Socket Index=&SpawnGroup()
    Inherit Velocity=false
    MaskSpawnGroup=false
    Mesh Sampling Type=::NewEnumerator5
    Sampled Position Offset=$System.OffsetBurstFromEngineBone
    Sampled Position Offset Coordinate Space:ENiagaraMeshTransforms=::NewEnumerator3
    Skeletal Mesh=$User.Skeletal Mesh
    Socket Sampling Mode=::NewEnumerator5
    Use Sampled Mesh Attribute as Ribbon ID=false
   )
   &FN_SetSpawnGroup(Number of Spawn Groups=2)
  ]
  particleUpdateScript [
   &ParticleState()
   /SetVariables(Particles.EndPosition:Vector3f=&Lerp_Vector(
     A=&RandomRangeVector(Minimum=<50 -20 -20>, Maximum=<250 50 20>)
     B=&RandomRangeVector(Minimum=<0 -50 -50>, Maximum=<200 50 50>)
     Alpha=&MakeFloatFromInt(Int=&SpawnGroup(Max Spawn Group=1))
    )) {isSetParameters=true}
   /SetVariables(Particles.SpriteAlignment=&FN_FrontEndVector3Modifier(
     Gameplay=$Particles.EndPosition
     Front End=/Game/Effects/Niagara/DynamicInputs/FrontEnd/FN_FrontEndRotateVector(Angle:NiagaraFloat=0.2, Vector to Rotate:Vector3f=$Particles.EndPosition)
    )) {isSetParameters=true, module=SetVariables_#7}
   &SkeletalMeshLocation(
    Bone / Socket Index=&SpawnGroup()
    Inherit Velocity=false
    MaskSpawnGroup=false
    Mesh Sampling Type=::NewEnumerator5
    Sampled Position Offset=&Add_Vector(
     A=$System.OffsetBurstFromEngineBone
     B=&Lerp_Vector(
      A=<0 0 0>
      B=&FN_FrontEndVector3Modifier(
       Gameplay=$Particles.EndPosition
       Front End=&FN_FrontEndRotateVector(Angle:NiagaraFloat=0.2, Vector to Rotate:Vector3f=$Particles.EndPosition)
      )
      Alpha=&Multiply_Float(A=$Particles.NormalizedAge, B=2)
     )
    )
    Sampled Position Offset Coordinate Space:ENiagaraMeshTransforms=::NewEnumerator3
    Skeletal Mesh=$User.Skeletal Mesh
    Socket Sampling Mode=::NewEnumerator5
    Use Sampled Mesh Attribute as Ribbon ID=false
   ) {module=SkeletalMeshLocation001}
   &ScaleColor(
    Scale Alpha=&FloatFromCurve(
     FloatCurve=DataInterface() {propertyValues {Curve {keys [<at 0 1>,<at 1 0>]}}}
     CurveIndex=&Multiply_Float(A=$Particles.NormalizedAge, B=2)
    )
    Scale RGB=&MakeVectorFromLinearColorRGB(Color=&ColorFromCurve(
      ColorCurve=DataInterface() {propertyValues {
        AlphaCurve {keys []}
        BlueCurve {keys [<at 0 0.25 RCIM_Cubic>,<at 0.5 0.5>,<at 1 0.875 RCIM_Cubic>]}
        GreenCurve {keys [<at 0 0.525 RCIM_Cubic>,<at 0.5 0.6833>,<at 1 0.9208 RCIM_Cubic>]}
        RedCurve {keys [<at 0 1 RCIM_Cubic>,<at 0.5 1>,<at 1 1 RCIM_Cubic>]}
       }}
      CurveIndex=&Multiply_Float(A=$Particles.NormalizedAge, B=2)
     ))
    ScaleA=true
    ScaleRGB=true
   )
   &ScaleColor(
    Scale RGB=&VectorFromFloat(Value=&FloatFromCurve(
      FloatCurve=DataInterface() {propertyValues {Curve {keys [
          <at 0 100 RCIM_Cubic>
          {arriveTangent=-36.3579, interpMode=RCIM_Cubic, leaveTangent=-36.3595, tangentMode=RCTM_User, time=0.25, value=10}
          <at 1 1 RCIM_Cubic>
         ]}}}
      CurveIndex=&Multiply_Float(A=$Particles.NormalizedAge, B=2)
     ))
    ScaleA=true
    ScaleRGB=true
   ) {module=ScaleColor001}
   &ScaleSpriteSize(Scale Factor=&Vector2DFromCurve(
     Vector2Curve=DataInterface() {propertyValues {
       XCurve {keys [<at 0 1>,<at 1 0>]}
       YCurve {keys [{arriveTangent=18.2053, interpMode=RCIM_Cubic, leaveTangent=18.2053, tangentMode=RCTM_User, time=0, value=1}, <at 1 10>]}
      }}
     CurveIndex=&Multiply_Float(A=$Particles.NormalizedAge, B=2)
     ScaleCurve=<1 1>
    ))
   /SetVariables(
    Particles.ColorAlt:LinearColor=<rgba 0 0 0 0.015>
    Particles.PositionAlt:Vector3f=&Lerp_Vector(A=$Output.SkeletalMeshLocation001.SampledPosition, B=Linked() {linkedVariable {name=Particles.Position, type=NiagaraPosition}}, Alpha=0.4)
    Particles.SpriteFacingAlt:Vector3f=<0 0 1>
    Particles.SpriteRotationAlt:NiagaraFloat=&RandomRangeFloat(Minimum=-360, Maximum=360)
    Particles.SpriteSizeAlt:Vector2f=&Vector2DFromFloat(Value=&Multiply_Float(
      A=&RandomRangeFloat(Minimum=60, Maximum=90)
      B=&Lerp_Float(
       A=1
       B=&RandomRangeFloat(Minimum=0, Maximum=1)
       Alpha=&MakeFloatFromInt(Int=$Particles.SpawnGroup)
      )
     ))
   ) {isSetParameters=true, module=SetVariables_#8}
  ]
  rendererClasses [&NiagaraSpriteRendererProperties]
  renderers [{rendererClass=&NiagaraSpriteRendererProperties}]
 }
 Sparks001 {
  bEnabled=false
  emitterUpdateScript [
   &EmitterState()
   &SpawnRate(
    Spawn Probability=&Lerp_Float(
     A=0.1
     B=&RandomRangeFloat(Minimum=0.05, Maximum=0.2)
     Alpha=&FN_FrontEndFloatModifier(Gameplay=$User.ForwardDot, Front End=1)
    )
    SpawnRate=50
    Use Spawn Probability=true
   )
   &SpawnRate(Spawn Probability=$System.GenerateForwardBurst, SpawnRate=200, Use Spawn Probability=true) {module=SpawnRate001}
  ]
  particleSpawnScript [&InitializeParticle(
    Color=<rgba 5 2.625 1.25 1>
    Lifetime=&RandomRangeFloat(Minimum=1, Maximum=2)
    Module.Write Color=true
    Module.Write Lifetime=true
    Module.Write Mass=true
    Module.Write Material Random=true
    Module.Write Position=false
    Module.Write Scale=false
    Module.Write SpriteRotation=false
    Module.Write SpriteSize=true
    Position=&SimulationPosition()
    Sprite Size=&Lerp_Vector2(
     A=&UniformRangedVector2D(Minimum=<0.5 1>, Maximum=<1 2>)
     B=&UniformRangedVector2D(Minimum=<1 1.5>, Maximum=<2 4>)
     Alpha=1
    )
   )]
  particleUpdateScript [
   &ParticleState()
   /SetVariables(Particles.StartPosition:Vector3f=&Lerp_Vector(
     A=&Add_Vector(
      A=<0 -16 0>
      B=&Multiply_VectorByFloat(
       Vector=/Niagara/DynamicInputs/Helpers/SafeNormalizeVector(
        Vector To Normalize:Vector3f=&RandomRangeVector(Minimum=<-1 0 -1>, Maximum=<1 0 1>)
        Fallback Vector:Vector3f=<0 0 0>
       )
       Float=22
      )
     )
     B=&Multiply_VectorByFloat(
      Vector=&SafeNormalizeVector(
       Vector To Normalize:Vector3f=&RandomRangeVector(Minimum=<-1 0 -1>, Maximum=<1 0 1>)
       Fallback Vector:Vector3f=<0 0 0>
      )
      Float=12
     )
     Alpha=&UniformAOrBFloat(A=0)
    )) {isSetParameters=true}
   /SetVariables(Particles.EndPosition:Vector3f=&Add_Vector(
     A=&Multiply_Vector(A=$Particles.StartPosition, B=<1.5 1 1.5>)
     B=&RandomRangeVector(Minimum=<0 50 0>, Maximum=<0 250 0>)
    )) {isSetParameters=true}
   /SetVariables(Particles.SpriteAlignment=&FN_FrontEndVector3Modifier(
     Gameplay=&Subtract_Vector(A=$Particles.EndPosition, B=$Particles.StartPosition)
     Front End=&FN_FrontEndRotateVector(
      Angle:NiagaraFloat=0.2
      Vector to Rotate:Vector3f=&Subtract_Vector(A=$Particles.EndPosition, B=$Particles.StartPosition)
     )
    )) {isSetParameters=true, module=SetVariables_#7}
   &SkeletalMeshLocation(
    Bone / Socket Index=&RandomRangeInt(Minimum=0, Maximum=1)
    Bone Sampling Mode=::NewEnumerator4
    Inherit Velocity=false
    Interpolate Mesh Sampling=false
    MaskSpawnGroup=false
    Sampled Position Offset=&Add_Vector(
     A=$System.OffsetBurstFromEngineBone
     B=&Lerp_Vector(A=$Particles.StartPosition, B=$Particles.EndPosition, Alpha=$Particles.NormalizedAge)
    )
    Sampled Position Offset Coordinate Space:ENiagaraMeshTransforms=::NewEnumerator3
    Skeletal Mesh=$User.Skeletal Mesh
    Use Sampled Mesh Attribute as Ribbon ID=false
   ) {module=SkeletalMeshLocation001}
   &ScaleColor(
    Scale Alpha=&FloatFromCurve(
     FloatCurve=DataInterface() {propertyValues {Curve {keys [<at 0 1>,<at 1 0>]}}}
     CurveIndex=&Multiply_Float(A=$Particles.NormalizedAge, B=2)
    )
    Scale RGB=&MakeVectorFromLinearColorRGB(Color=&ColorFromCurve(
      ColorCurve=DataInterface() {propertyValues {
        AlphaCurve {keys []}
        BlueCurve {keys [<at 0 0.25 RCIM_Cubic>,<at 0.5 0.5>,<at 1 0.875 RCIM_Cubic>]}
        GreenCurve {keys [<at 0 0.525 RCIM_Cubic>,<at 0.5 0.6833>,<at 1 0.9208 RCIM_Cubic>]}
        RedCurve {keys [<at 0 1 RCIM_Cubic>,<at 0.5 1>,<at 1 1 RCIM_Cubic>]}
       }}
      CurveIndex=&Multiply_Float(A=$Particles.NormalizedAge, B=2)
     ))
    ScaleA=true
    ScaleRGB=true
   )
   &ScaleColor(
    Scale RGB=&VectorFromFloat(Value=&FloatFromCurve(
      FloatCurve=DataInterface() {propertyValues {Curve {keys [
          <at 0 100 RCIM_Cubic>
          {arriveTangent=-36.3579, interpMode=RCIM_Cubic, leaveTangent=-36.3595, tangentMode=RCTM_User, time=0.25, value=10}
          <at 1 1 RCIM_Cubic>
         ]}}}
      CurveIndex=&Multiply_Float(A=$Particles.NormalizedAge, B=2)
     ))
    ScaleA=true
    ScaleRGB=true
   ) {module=ScaleColor001}
   &ScaleSpriteSize(Scale Factor=&Vector2DFromCurve(
     Vector2Curve=DataInterface() {propertyValues {
       XCurve {keys [<at 0 1>,<at 1 0>]}
       YCurve {keys [{arriveTangent=18.2053, interpMode=RCIM_Cubic, leaveTangent=18.2053, tangentMode=RCTM_User, time=0, value=1}, <at 1 10>]}
      }}
     CurveIndex=&Multiply_Float(A=$Particles.NormalizedAge, B=2)
     ScaleCurve=<20 20>
    ))
   /SetVariables(
    Particles.ColorAlt:LinearColor=<rgba 0 0 0 0.015>
    Particles.PositionAlt:Vector3f=&Lerp_Vector(A=$Output.SkeletalMeshLocation001.SampledPosition, B=Linked() {linkedVariable {name=Particles.Position, type=NiagaraPosition}}, Alpha=0.4)
    Particles.SpriteFacingAlt:Vector3f=<0 0 1>
    Particles.SpriteRotationAlt:NiagaraFloat=&RandomRangeFloat(Minimum=-360, Maximum=360)
    Particles.SpriteSizeAlt:Vector2f=&Vector2DFromFloat(Value=&Multiply_Float(
      A=&RandomRangeFloat(Minimum=60, Maximum=90)
      B=&Lerp_Float(
       A=1
       B=&RandomRangeFloat(Minimum=0, Maximum=1)
       Alpha=&MakeFloatFromInt(Int=$Particles.SpawnGroup)
      )
     ))
   ) {isSetParameters=true, module=SetVariables_#8}
  ]
  rendererClasses [&NiagaraSpriteRendererProperties]
  renderers [{rendererClass=&NiagaraSpriteRendererProperties}]
 }
 Start_Burst {
  bEnabled=false
  emitterUpdateScript [
   &EmitterState()
   &SpawnRate(SpawnRate=&Multiply_Float(A=50, B=$User.ForwardBurst), Use Spawn Probability=false)
   &SpawnBurst_Instantaneous(
    Spawn Count=&MakeCustomIntFromBool(Bool=$User.bIsFrontEndPreview, True INT=5, False INT=0)
    Spawn Probability=&MakeCustomFloatFromBool(
     Bool=&SetBoolByIntComparison(A=$Emitter.LoopCount, B=0, Comparison Type=::NewEnumerator4)
     True Float=1
     False Float=0
    )
    Use Spawn Probability=true
   )
  ]
  particleSpawnScript [
   &FN_SetSpawnGroup_v2(Number of Spawn Groups=2)
   &InitializeParticle(
    Color=<rgba 1.5 0.55 0 0.1>
    Lifetime=&RandomRangeFloat(Minimum=0.1, Maximum=0.2)
    Mesh Scale=<1 1 4>
    Module.Write Color=true
    Module.Write Lifetime=true
    Module.Write Mass=true
    Module.Write Material Random=false
    Module.Write Position=false
    Module.Write Scale=true
    Module.Write SpriteRotation=false
    Module.Write SpriteSize=true
    Position=&SimulationPosition()
    Sprite Size=&UniformRangedVector2D(Minimum=<100 4>, Maximum=<50 1>)
   )
   /SetVariables(
    Particles.CameraOffset_2=10
    Particles.Color_2:LinearColor=<rgba 1 0.3667 0 1>
    Particles.DynamicMaterialParameter_2:Vector4f=<1 1 1 1>
    Particles.SpriteSize_2:Vector2f=<100 100>
   ) {isSetParameters=true}
   /SetVariables(
    Particles.Color_3:LinearColor=<rgba 5 5 5 1>
    Particles.DynamicMaterialParameter_3:Vector4f=&MakeVector4(X=0.5, Y=1, Z=&UniformRangedFloat(Minimum=0.25, Maximum=0.35), W=1)
    Particles.SpriteSize_3:Vector2f=&Vector2DFromFloat(Value=&UniformRangedFloat(Minimum=50, Maximum=75))
   ) {isSetParameters=true}
   /SetVariables(
    Particles.Color_4:LinearColor=<rgba 1.5 0.7875 0.375 0.1>
    Particles.DynamicMaterialParameter_4:Vector4f=&MakeVector4(X=1, Y=5, Z=20, W=1)
   ) {isSetParameters=true}
   &CameraOffset(Camera Offset Amount=2)
   &DynamicMaterialParameters(
    Index 0 Param 1=5
    Module.Param0WriteEnabled=true
    Module.Param1WriteEnabled=false
    Module.Param2WriteEnabled=false
    Module.Param3WriteEnabled=false
   )
   &InitialMeshOrientation(Orientation Vector=<0 1 0>, Use Orientation Vector=true, Use Rotation Vector=false) {enabled=false}
   /Niagara/Modules/Spawn/Orientation/V3/InitialMeshOrientation(
    Mesh Axis To Orient=&RotateVectorByQuaternion(Vector To Rotate=<1 0 0>, Quaternion=&MakeQuatFromAxisAngle(Axis=<0 1 0>, AngleInDegrees=-15))
    Mesh Orientation Mode=::NewEnumerator2
    Orientation Coordinate Space=::Local
    Orientation Vector=<0 1 0>
    Use Rotation Vector=false
   ) {module=InitialMeshOrientation001}
  ]
  particleUpdateScript [
   &ParticleState()
   &SkeletalMeshLocation(
    Bone / Socket Index=&SpawnGroup()
    Inherit Velocity=false
    MaskSpawnGroup=false
    Mesh Sampling Type=::NewEnumerator5
    Sampled Position Offset=$System.OffsetBurstFromEngineBone
    Sampled Position Offset Coordinate Space:ENiagaraMeshTransforms=::NewEnumerator3
    Skeletal Mesh=$User.Skeletal Mesh
    Socket Sampling Mode=::NewEnumerator5
   )
   &ScaleSpriteSize(Scale Factor=&MakeVector2D(X=&FloatFromCurve(FloatCurve=DataInterface() {propertyValues {Curve {keys [<at 0 0>,<at 1 2>]}}}), Y=1))
   /SetVariables(Particles.SpriteSize_3:Vector2f=&Multiply_Vector2(
     A=$Particles.Initial.SpriteSize_3
     B=&Vector2DFromFloat(Value=&FloatFromCurve(FloatCurve=DataInterface() {propertyValues {Curve {keys [<at 0 0>,<at 1 1>]}}}))
    )) {isSetParameters=true}
   &ScaleColor(
    Scale Alpha=&FloatFromCurve(FloatCurve=DataInterface() {propertyValues {Curve {keys [<at 0 1>,<at 1 0>]}}})
    Scale RGB=&MakeVectorFromLinearColorRGB(Color=&ColorFromCurve(ColorCurve=DataInterface() {propertyValues {
        AlphaCurve {keys []}
        BlueCurve {keys [<at 0 0.25 RCIM_Cubic>,<at 0.5 0.5>,<at 1 0.875 RCIM_Cubic>]}
        GreenCurve {keys [<at 0 0.525 RCIM_Cubic>,<at 0.5 0.6833>,<at 1 0.9208 RCIM_Cubic>]}
        RedCurve {keys [<at 0 1 RCIM_Cubic>,<at 0.5 1>,<at 1 1 RCIM_Cubic>]}
       }}))
    ScaleA=true
    ScaleRGB=true
   )
   &ScaleColor(
    Scale RGB=&VectorFromFloat(Value=&FloatFromCurve(FloatCurve=DataInterface() {propertyValues {Curve {keys [
          <at 0 100 RCIM_Cubic>
          {arriveTangent=-33.2952, interpMode=RCIM_Cubic, leaveTangent=-33.2924, tangentMode=RCTM_User, time=0.25, value=10}
          <at 1 0 RCIM_Cubic>
         ]}}}))
    ScaleA=true
    ScaleRGB=true
   ) {module=ScaleColor001}
   /SetVariables(
    Particles.Color_2:LinearColor=&MultiplyLinearColor(
     Linear Color=$Particles.Initial.Color_2
     Scale Factor=&MakeVector4_FromLinearColor(LinearColor=&MakeLinearColorFromVectorAndFloat(
       "Vector (RGB)"=&MakeVectorFromLinearColorRGB(Color=&ColorFromCurve(ColorCurve=DataInterface() {propertyValues {
           AlphaCurve {keys []}
           BlueCurve {keys [<at 0 0.25 RCIM_Cubic>,<at 0.5 0.5>,<at 1 0.875 RCIM_Cubic>]}
           GreenCurve {keys [<at 0 0.525 RCIM_Cubic>,<at 0.5 0.6833>,<at 1 0.9208 RCIM_Cubic>]}
           RedCurve {keys [<at 0 1 RCIM_Cubic>,<at 0.5 1>,<at 1 1 RCIM_Cubic>]}
          }}))
       "Float (Alpha)"=&FloatFromCurve(FloatCurve=DataInterface() {propertyValues {Curve {keys [<at 0 1>,<at 1 0>]}}})
      ))
    )
    Particles.Color_3:LinearColor=&MultiplyLinearColor(
     Linear Color=$Particles.Initial.Color_3
     Scale Factor=&MakeVector4_FromLinearColor(LinearColor=&MakeLinearColorFromVectorAndFloat(
       "Vector (RGB)"=&MakeVectorFromLinearColorRGB(Color=&ColorFromCurve(ColorCurve=DataInterface() {propertyValues {
           AlphaCurve {keys []}
           BlueCurve {keys [<at 0 0.25 RCIM_Cubic>,<at 0.5 0.5>,<at 1 0.875 RCIM_Cubic>]}
           GreenCurve {keys [<at 0 0.525 RCIM_Cubic>,<at 0.5 0.6833>,<at 1 0.9208 RCIM_Cubic>]}
           RedCurve {keys [<at 0 1 RCIM_Cubic>,<at 0.5 1>,<at 1 1 RCIM_Cubic>]}
          }}))
       "Float (Alpha)"=&FloatFromCurve(FloatCurve=DataInterface() {propertyValues {Curve {keys [<at 0 1>,<at 1 0>]}}})
      ))
    )
    Particles.Color_4:LinearColor=&MultiplyLinearColor(
     Linear Color=$Particles.Initial.Color_4
     Scale Factor=&MakeVector4_FromLinearColor(LinearColor=&MakeLinearColorFromVectorAndFloat(
       "Vector (RGB)"=&MakeVectorFromLinearColorRGB(Color=&ColorFromCurve(ColorCurve=DataInterface() {propertyValues {
           AlphaCurve {keys []}
           BlueCurve {keys [<at 0 0.25 RCIM_Cubic>,<at 0.5 0.5>,<at 1 0.875 RCIM_Cubic>]}
           GreenCurve {keys [<at 0 0.525 RCIM_Cubic>,<at 0.5 0.6833>,<at 1 0.9208 RCIM_Cubic>]}
           RedCurve {keys [<at 0 1 RCIM_Cubic>,<at 0.5 1>,<at 1 1 RCIM_Cubic>]}
          }}))
       "Float (Alpha)"=&FloatFromCurve(FloatCurve=DataInterface() {propertyValues {Curve {keys [<at 0 1>,<at 1 0>]}}}, Scale Curve=3)
      ))
    )
   ) {isSetParameters=true}
   /SetVariables(
    Particles.Color_2:LinearColor=&MultiplyLinearColor(
     Linear Color=$Particles.Color_2
     Scale Factor=&MakeFloat4FromVectorAndFloat(
      "Vector (RGB)"=&VectorFromFloat(Value=&FloatFromCurve(FloatCurve=DataInterface() {propertyValues {Curve {keys [<at 0 1 RCIM_Cubic>, {arriveTangent=-1, interpMode=RCIM_Cubic, leaveTangent=-1, time=0.1, value=6}, <at 1 0 RCIM_Cubic>]}}}))
      "Float (Alpha)"=1
     )
    )
    Particles.Color_4:LinearColor=&MultiplyLinearColor(
     Linear Color=$Particles.Initial.Color_4
     Scale Factor=&MakeFloat4FromVectorAndFloat(
      "Vector (RGB)"=&VectorFromFloat(Value=&FloatFromCurve(FloatCurve=DataInterface() {propertyValues {Curve {keys [
            <at 0 100 RCIM_Cubic>
            {arriveTangent=-36.3579, interpMode=RCIM_Cubic, leaveTangent=-36.3595, tangentMode=RCTM_User, time=0.25, value=10}
            <at 1 1 RCIM_Cubic>
           ]}}}))
      "Float (Alpha)"=1
     )
    )
   ) {isSetParameters=true}
   &ScaleMeshSize(Scale Factor=&VectorFromCurve(
     VectorCurve=DataInterface() {propertyValues {
       XCurve {keys [
         {arriveTangent=2.9897, interpMode=RCIM_Cubic, leaveTangent=2.9897, tangentMode=RCTM_User, time=0, value=0}
         {arriveTangent=3.9553, interpMode=RCIM_Cubic, leaveTangent=3.9553, tangentMode=RCTM_User, time=0.2, value=0.6}
         <at 1 2 RCIM_Cubic>
        ]}
       YCurve {keys [
         {arriveTangent=14.017, interpMode=RCIM_Cubic, leaveTangent=14.017, tangentMode=RCTM_User, time=0, value=0}
         {arriveTangent=0.75, interpMode=RCIM_Cubic, leaveTangent=0.75, time=0.2, value=1.5}
         <at 1 0.75 RCIM_Cubic>
        ]}
       ZCurve {keys [
         {arriveTangent=7.6324, interpMode=RCIM_Cubic, leaveTangent=7.6324, tangentMode=RCTM_User, time=0, value=0}
         <at 0.2 0.65 RCIM_Cubic>
         <at 1 0 RCIM_Cubic>
        ]}
      }}
     Scale Curve=&RandomRangeVector(Minimum=<1 0.75 0.5>, Maximum=<1 1 0.75>)
    ))
  ]
  rendererClasses [&NiagaraSpriteRendererProperties, &NiagaraMeshRendererProperties]
  renderers [
   {rendererClass=&NiagaraSpriteRendererProperties}
   {rendererClass=&NiagaraSpriteRendererProperties, rendererIndex=1}
   {rendererClass=&NiagaraSpriteRendererProperties, rendererIndex=2}
   {rendererClass=&NiagaraMeshRendererProperties, rendererIndex=3}
  ]
 }
 ThrusterSparks {
  simTarget=GPUComputeSim
  emitterUpdateScript [
   &EmitterState(Life Cycle Mode=::NewEnumerator1, Loop Behavior=::NewEnumerator1, Loop Duration=1.2, UseLoopDelay=false)
   &SpawnRate(
    Spawn Probability=&Lerp_Float(
     A=0.1
     B=&RandomRangeFloat(Minimum=0.05, Maximum=0.2)
     Alpha=&FN_FrontEndFloatModifier(Gameplay=$User.ForwardDot, Front End=1)
    )
    SpawnRate=50
    Use Spawn Probability=true
   ) {enabled=false}
   &SpawnRate(Spawn Probability=$System.GenerateForwardBurst, SpawnRate=200, Use Spawn Probability=true) {enabled=false, module=SpawnRate001}
   &SpawnRate(SpawnRate=1000, Use Spawn Probability=false) {module=SpawnRate002}
  ]
  particleSpawnScript [
   /Niagara/Modules/Spawn/Initialization/V2/InitializeParticle(
    Color=$System.SysColorA
    Lifetime=0.03
    Sprite Size=<1 10>
    Sprite Size Mode=::NewEnumerator4
    UsePositionOffset=false
    Write Material Random=true
   )
   &AddVelocityFromPoint(Module.Use Falloff Distance=false, Origin Offset=<0 -100 0>, Velocity Strength=500) {enabled=false}
   &SkeletalMeshLocation(
    Bone / Socket Index=&RandomRangeInt(Minimum=0, Maximum=1)
    MaskSpawnGroup=false
    Mesh Sampling Type=::NewEnumerator5
    Orientation Sampling=::NewEnumerator1
    Skeletal Mesh=$User.Skeletal Mesh
    Socket Sampling Mode=::NewEnumerator5
   ) {module=SkeletalMeshLocation001}
   &TorusLocation(
    Align Torus With Axis=true
    Coordinate Space:ENiagaraCoordinateSpace=::Simulation
    Large Radius=5
    MaskSpawnGroup=false
    Module.Override Randomness=false
    Surface Position=1
    Torus Axis=&RotateVectorByQuaternion(Vector To Rotate=<1 0 0>, Quaternion=$Particles.SkeletalMeshLocation001.SampledRotation)
    Torus Mode=::NewEnumerator2
   )
   &AddVelocity(
    Coordinate Space=::Simulation
    Velocity=&RotateVectorByQuaternion(Vector To Rotate=<400 0 0>, Quaternion=$Particles.SkeletalMeshLocation001.SampledRotation)
   ) {enabled=false}
   &AddVelocity(
    Coordinate Space=::Simulation
    Velocity=&Multiply_VectorByFloat(Vector=$Output.TorusLocation.TorusBitangent, Float=800)
   ) {module=AddVelocity002}
   &AddVelocity(
    Coordinate Space=::Simulation
    Velocity=&Multiply_VectorByFloat(Vector=$Output.TorusLocation.TorusNormal, Float=200)
   ) {module=AddVelocity001}
  ]
  particleUpdateScript [
   &SolveForcesAndVelocity(Clamp Velocity=false, Limit Acceleration=false)
   &ParticleState()
   /SetVariables() {enabled=false, isSetParameters=true}
   /SetVariables(Particles.EndPosition:Vector3f=&Add_Vector(
     A=&Multiply_Vector(
      A=$Particles.StartPosition
      B=/Niagara/DynamicInputs/Transforms/TransformPosition(Position=<200 0 0>, Source Space=::World, Destination Space=::Local)
     )
     B=<1 1 1>
    )) {enabled=false, isSetParameters=true}
   /SetVariables(Particles.StartPosition:Vector3f=&Lerp_Vector(
     A=&Add_Vector(
      A=<0 -16 0>
      B=&Multiply_VectorByFloat(
       Vector=&SafeNormalizeVector(
        Vector To Normalize:Vector3f=&RandomRangeVector(Minimum=<-1 0 -1>, Maximum=<1 0 1>)
        Fallback Vector:Vector3f=<0 0 0>
       )
       Float=22
      )
     )
     B=&Multiply_VectorByFloat(
      Vector=&SafeNormalizeVector(
       Vector To Normalize:Vector3f=&RandomRangeVector(Minimum=<-1 0 -1>, Maximum=<1 0 1>)
       Fallback Vector:Vector3f=<0 0 0>
      )
      Float=12
     )
     Alpha=&UniformAOrBFloat(A=0)
    )) {enabled=false, isSetParameters=true}
   /SetVariables(Particles.EndPosition:Vector3f=&Add_Vector(
     A=&Multiply_Vector(A=$Particles.StartPosition, B=<1.5 1 1.5>)
     B=&RandomRangeVector(Minimum=<250 0 50>, Maximum=<250 0 50>)
    )) {enabled=false, isSetParameters=true}
   /SetVariables(Particles.SpriteAlignment=&FN_FrontEndVector3Modifier(
     Gameplay=&Subtract_Vector(A=$Particles.EndPosition, B=$Particles.StartPosition)
     Front End=&FN_FrontEndRotateVector(
      Angle:NiagaraFloat=0.2
      Vector to Rotate:Vector3f=&Subtract_Vector(A=$Particles.EndPosition, B=$Particles.StartPosition)
     )
    )) {enabled=false, isSetParameters=true}
   &SkeletalMeshLocation(
    Bone / Socket Index=&RandomRangeInt(Minimum=0, Maximum=1)
    MaskSpawnGroup=false
    Mesh Sampling Type=::NewEnumerator5
    Sampled Position Offset=&Add_Vector(
     A=$System.OffsetBurstFromEngineBone
     B=&Lerp_Vector(A=$Particles.StartPosition, B=$Particles.EndPosition, Alpha=0)
    )
    Skeletal Mesh=$User.Skeletal Mesh
    Socket Sampling Mode=::NewEnumerator5
   ) {enabled=false}
   &ScaleColor(
    Scale Alpha=&FloatFromCurve(
     FloatCurve=DataInterface() {propertyValues {Curve {keys [<at 0 1>,<at 1 0>]}}}
     CurveIndex=&Multiply_Float(A=$Particles.NormalizedAge, B=2)
    )
    Scale RGB=&MakeVectorFromLinearColorRGB(Color=&ColorFromCurve(
      ColorCurve=DataInterface() {propertyValues {
        AlphaCurve {keys []}
        BlueCurve {keys [<at 0 0.25 RCIM_Cubic>,<at 0.5 0.5>,<at 1 0.875 RCIM_Cubic>]}
        GreenCurve {keys [<at 0 0.525 RCIM_Cubic>,<at 0.5 0.6833>,<at 1 0.9208 RCIM_Cubic>]}
        RedCurve {keys [<at 0 1 RCIM_Cubic>,<at 0.5 1>,<at 1 1 RCIM_Cubic>]}
       }}
      CurveIndex=&Multiply_Float(A=$Particles.NormalizedAge, B=1)
     ))
    ScaleA=true
    ScaleRGB=true
   )
   &ScaleColor(
    Scale RGB=&VectorFromFloat(Value=&FloatFromCurve(
      FloatCurve=DataInterface() {propertyValues {Curve {keys [
          <at 0 100 RCIM_Cubic>
          {arriveTangent=-36.3579, interpMode=RCIM_Cubic, leaveTangent=-36.3595, tangentMode=RCTM_User, time=0.25, value=10}
          <at 1 1 RCIM_Cubic>
         ]}}}
      CurveIndex=&Multiply_Float(A=$Particles.NormalizedAge, B=2)
     ))
    ScaleA=true
    ScaleRGB=true
   ) {module=ScaleColor001}
   &ScaleSpriteSize(Scale Factor=&Vector2DFromCurve(
     Vector2Curve=DataInterface() {propertyValues {
       XCurve {keys [<at 0 1>,<at 1 0>]}
       YCurve {keys [{arriveTangent=18.2053, interpMode=RCIM_Cubic, leaveTangent=18.2053, tangentMode=RCTM_User, time=0, value=1}, <at 1 10>]}
      }}
     CurveIndex=&Multiply_Float(A=$Particles.NormalizedAge, B=2)
     ScaleCurve=<1 1>
    ))
   &SpriteFacingAndAlignment(WriteAlignment=false, WriteFacing=true) {enabled=false}
  ]
  rendererClasses [&NiagaraSpriteRendererProperties, &NiagaraMeshRendererProperties]
  renderers [
   {rendererClass=&NiagaraSpriteRendererProperties}
   {rendererClass=&NiagaraMeshRendererProperties, rendererIndex=1}
  ]
 }
}
summary {
 systemName=NS_Glider_Noble_JetBurst
 emitters [
  {bEnabled=false, emitterName=Lightning, rendererClasses [&NiagaraMeshRendererProperties]}
  {
   bEnabled=false
   emitterName=Start_Burst
   rendererClasses [&NiagaraSpriteRendererProperties, &NiagaraMeshRendererProperties]
  }
  {bEnabled=false, emitterName=Sparks, simTarget=GPUComputeSim, rendererClasses [&NiagaraSpriteRendererProperties]}
  {bEnabled=false, emitterName=Flame_Mesh, rendererClasses [&NiagaraMeshRendererProperties]}
  {emitterName=Light, rendererClasses [&NiagaraSpriteRendererProperties, &NiagaraLightRendererProperties]}
  {bEnabled=false, emitterName=MeshTrails, rendererClasses [&NiagaraMeshRendererProperties]}
  {bEnabled=false, emitterName=Sparks001, rendererClasses [&NiagaraSpriteRendererProperties]}
  {
   emitterName=ThrusterSparks
   simTarget=GPUComputeSim
   rendererClasses [&NiagaraSpriteRendererProperties, &NiagaraMeshRendererProperties]
  }
  {bEnabled=false, emitterName=Flame_Mesh001, rendererClasses [&NiagaraMeshRendererProperties]}
  {bEnabled=false, emitterName=Flame_MeshCones, rendererClasses [&NiagaraMeshRendererProperties]}
  {bEnabled=false, emitterName=Flame_MeshOval, rendererClasses [&NiagaraMeshRendererProperties]}
  {bEnabled=false, emitterName=Empty001, rendererClasses [&NiagaraSpriteRendererProperties]}
  {emitterName=Big_Old_Flares, rendererClasses [&NiagaraSpriteRendererProperties]}
  {emitterName=Flame_Mesh003, rendererClasses [&NiagaraMeshRendererProperties]}
  {emitterName=Flame_Mesh004, rendererClasses [&NiagaraMeshRendererProperties]}
  {emitterName=Flame_Mesh002, rendererClasses [&NiagaraMeshRendererProperties]}
  {emitterName=Light001, rendererClasses [&NiagaraSpriteRendererProperties, &NiagaraLightRendererProperties]}
 ]
 userVariables [
  User.bIsFrontEnd {type=NiagaraBool}
  User.bIsFrontEndPreview {type=NiagaraBool}
  User.BPOpacityMult {type=NiagaraFloat, defaultValue=0, description=""}
  User.ColorA {type=LinearColor, description="", defaultValue {struct=LinearColor, value=<rgba 0 0.3333 2 1>}}
  User.ColorB {type=LinearColor, description="", defaultValue {struct=LinearColor, value=<rgba 0.6 1.0667 2 1>}}
  User.ColorHueShift {type=NiagaraFloat, defaultValue=200, description=""}
  User.CurrentExhaust {type=Vector3f, description="", defaultValue {struct=Vector3f, value=<0 0 0>}}
  User.emissive mult {type=NiagaraFloat, defaultValue=0, description=""}
  User.ForwardBurst {type=NiagaraFloat, defaultValue=1, description=""}
  User.ForwardDot {type=NiagaraFloat, defaultValue=-1, description=""}
  User.PlayerSpeed {type=NiagaraFloat, defaultValue=0, description=""}
  User.RightDot {type=NiagaraFloat, defaultValue=0, description=""}
  User.Skeletal Mesh {
   type=NiagaraDataInterfaceSkeletalMesh
   defaultValue {
    struct=NiagaraExt_VariableValue_DataInterface
    value {dataInterface=&self:NiagaraDataInterfaceSkeletalMesh_2}
   }
  }
  User.ThrusterPower {type=NiagaraFloat, defaultValue=0, description=""}
  User.TrailLengthScale {type=NiagaraFloat, defaultValue=0, description=""}
 ]
}
