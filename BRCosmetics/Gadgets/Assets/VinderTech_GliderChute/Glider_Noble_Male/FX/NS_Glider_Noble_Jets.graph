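# Niagara system for the blue jet exhaust on the male Noble VinderTech glider chute.
# Two mesh thruster emitters, two sprite flare emitters and a particle light emitter track skeletal mesh bones or sockets through PartitionParticles and SkeletalMeshLocation.
# Thruster power drives plume scale, opacity, flare size and light intensity, while ForwardDot drives plume lifetime.
# FrontEnd modifier functions substitute separate values for front-end display versus gameplay.
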NiagaraSystem /BRCosmetics/Gadgets/Assets/VinderTech_GliderChute/Glider_Noble_Male/FX/NS_Glider_Noble_Jets
dependencies {
 usedDataInterfaces [
  &NiagaraDataInterfaceCurve
  /Script/Niagara.NiagaraDataInterfaceSimpleCounter
  &NiagaraDataInterfaceArrayFloat3
  &NiagaraDataInterfaceVectorCurve
  &NiagaraDataInterfaceColorCurve
  &NiagaraDataInterfaceArrayFloat
 ]
 usedDynamicInputs [
  &MakeCustomVectorFromBool
  &Boolean_AndOperation
  &SetBoolByFloatComparison
  &MakeCustomFloatFromBool
  &HueShiftLinearColor
  &Divide_Float
 ]
 usedModules [
  &self:SystemSpawnScript.SystemScriptSource.SystemScriptGraph.NiagaraNodeAssignment_0.SetVariables_#1
  &FN_Integrator
  &self:SystemSpawnScript.SystemScriptSource.SystemScriptGraph.NiagaraNodeAssignment_1.SetVariables_#2
  &SystemState
  &self:Set_Current_Exhaust
  &self:Set_Trail_Length
  &self:Set_Trail_Opacity
  &Timeline
  &self:SystemSpawnScript.SystemScriptSource.SystemScriptGraph.NiagaraNodeAssignment_7.SetVariables_#3
  &self:SystemSpawnScript.SystemScriptSource.SystemScriptGraph.NiagaraNodeAssignment_2.SetVariables_#4
 ]
}
emitters {
 AmbientThruster {
  emitterUpdateScript [
   &EmitterState()
   &SpawnRate(
    SpawnRate=&Lerp_Float(A=30, B=40, Alpha=&FN_FrontEndFloatModifier(Gameplay=$System.ThrusterPower, Front End=1))
    Use Spawn Probability=false
   )
  ]
  particleSpawnScript [
   &PartitionParticles()
   &SkeletalMeshLocation(
    Bone / Socket Index=$Particles.PartitionParticles.Partition
    MaskSpawnGroup=false
    Mesh Sampling Type=::NewEnumerator5
    Orientation Sampling=::NewEnumerator1
    Sampled Position Offset=<0 0 1.5>
    Skeletal Mesh=$User.Skeletal Mesh
    Socket Sampling Mode=::NewEnumerator5
   ) {module=SkeletalMeshLocation001}
   /SetVariables(
    Particles.PlacementPicker:NiagaraInt32=/Niagara/DynamicInputs/Divide/Divide_Int(
     A:NiagaraInt32=&ModuloInt(
      A=/Niagara/DynamicInputs/DataInterface/GenerateIntFromCounter(SimpleCounter:NiagaraDataInterfaceSimpleCounter=DataInterface() {propertyValues {GpuSyncMode=None, InitialValue=0}})
      B=4
     )
     B:NiagaraInt32=2
    )
    Particles.SpawnGroup=&ModuloInt(
     A=&GenerateIntFromCounter(SimpleCounter:NiagaraDataInterfaceSimpleCounter=DataInterface() {propertyValues {GpuSyncMode=None, InitialValue=0}})
     B=2
    )
   ) {isSetParameters=true}
   /SetVariables(Particles.ShowBackFire:NiagaraInt32=/Niagara/DynamicInputs/Subtract/Subtract_Integer(A:NiagaraInt32=1, B:NiagaraInt32=$Particles.PlacementPicker)) {isSetParameters=true}
   /Niagara/Modules/Spawn/Initialization/InitializeParticle(
    Color=<rgba 1 1 1 0.05>
    Lifetime=&Lerp_Float(A=0.35, B=0.5, Alpha=&FN_FrontEndFloatModifier(Gameplay=$User.ForwardDot, Front End=0.02))
    Mesh Scale=&Multiply_Vector(
     A=&SelectVectorFromArray(
      Direct Array Index=$Particles.PlacementPicker
      Vector Selection Array=DataInterface() {propertyValues {FloatData [<0.15 0.15 0.03>, <0.15 0.15 0.06>]}}
     )
     B=&Lerp_Vector(
      A=<1 1 1>
      B=&RandomRangeVector(Minimum=<0.5 0.5 1>, Maximum=<1 1 1.5>)
      Alpha=&FN_FrontEndFloatModifier(Gameplay=$System.ThrusterPower, Front End=1)
     )
    )
    Module.Write Color=true
    Module.Write Lifetime=true
    Module.Write Mass=false
    Module.Write Material Random=false
    Module.Write Position=false
    Module.Write Scale=true
    Module.Write SpriteRotation=false
    Module.Write SpriteSize=false
    Position=&SimulationPosition()
    Sprite Size=<5 5>
   )
   &InitialMeshOrientation(
    Mesh Axis To Orient=<0 0 1>
    Mesh Orientation Mode=::NewEnumerator2
    Orientation Vector=&RotateVectorByQuaternion(Vector To Rotate=<1 0 0>, Quaternion=$Particles.SkeletalMeshLocation001.SampledRotation)
    Rotation=&RandomRangeVector(Minimum=<0 0 0>, Maximum=<0 0 1>)
    Use Rotation Vector=true
   )
  ]
  particleUpdateScript [
   &ParticleState()
   &SkeletalMeshLocation(
    Bone / Socket Index=$Particles.PartitionParticles.Partition
    MaskSpawnGroup=false
    Mesh Sampling Type=::NewEnumerator5
    Orientation Sampling=::NewEnumerator1
    Sampled Position Offset=<-10 0 0>
    Skeletal Mesh=$User.Skeletal Mesh
    Socket Sampling Mode=::NewEnumerator5
   )
   &ScaleMeshSize(Scale Factor=&FN_FrontEndVector3Modifier(
     Gameplay=&VectorFromCurve(
      VectorCurve=DataInterface() {propertyValues {
        XCurve {keys [<at 0 1>,<at 1 1>]}
        YCurve {keys [<at 0 1>,<at 1 1>]}
        ZCurve {keys [<at 0.9 1 RCIM_Cubic>, {arriveTangent=-4.873, interpMode=RCIM_Cubic, leaveTangent=-4.873, tangentMode=RCTM_User, time=1, value=0}]}
       }}
      Scale Curve=&Lerp_Vector(
       A=<0.85 0.85 0.5>
       B=<0.85 0.85 1.2>
       Alpha=&FN_FrontEndFloatModifier(Gameplay=$System.ThrusterPower, Front End=1)
      )
     )
     Front End=<1 1 2>
    ))
   &ScaleColor(
    Scale Alpha=&Lerp_Float(
     A=0
     B=&Lerp_Float(A=0.2, B=1, Alpha=&FN_FrontEndFloatModifier(Gameplay=$System.ThrusterPower, Front End=0.2))
     Alpha=$Particles.NormalizedAge
    )
    Scale RGB=&MakeVectorFromLinearColorRGB(Color=&ColorFromCurve(ColorCurve=DataInterface() {propertyValues {
        AlphaCurve {keys [<at 0 1 RCIM_Cubic>,<at 1 1 RCIM_Cubic>]}
        BlueCurve {keys [<at 0.2301 1 RCIM_Cubic>,<at 0.5656 0.1 RCIM_Cubic>]}
        GreenCurve {keys [<at 0.2301 0.2605 RCIM_Cubic>,<at 0.5656 0.0167 RCIM_Cubic>]}
        RedCurve {keys [<at 0.2301 0 RCIM_Cubic>,<at 0.5656 0 RCIM_Cubic>]}
       }}))
    ScaleA=true
    ScaleRGB=false
   )
   /SetVariables(Particles.ColorAlt:LinearColor=&MakeLinearColorFromVectorAndFloat(
     "Vector (RGB)"=&MakeVectorFromLinearColorRGB(Color=&ColorFromCurve(ColorCurve=DataInterface() {propertyValues {
         AlphaCurve {keys [<at 0 1 RCIM_Cubic>,<at 1 1 RCIM_Cubic>]}
         BlueCurve {keys [<at 0.2301 1 RCIM_Cubic>,<at 0.5656 0.1 RCIM_Cubic>]}
         GreenCurve {keys [<at 0.2301 0.2605 RCIM_Cubic>,<at 0.5656 0.0167 RCIM_Cubic>]}
         RedCurve {keys [<at 0.2301 0 RCIM_Cubic>,<at 0.5656 0 RCIM_Cubic>]}
        }}))
     "Float (Alpha)"=&Lerp_Float(
      A=0
      B=&Lerp_Float(A=0.02, B=0.5, Alpha=&FN_FrontEndFloatModifier(Gameplay=$System.ThrusterPower, Front End=0.5))
      Alpha=$Particles.NormalizedAge
     )
    )) {isSetParameters=true}
   &DynamicMaterialParameters(
    Index 0 Param 1=1
    Index 0 Param 2=0.3
    Index 0 Param 4=&Lerp_Float(A=5, B=20, Alpha=&FN_FrontEndFloatModifier(Gameplay=$System.ThrusterPower, Front End=1))
    Index 1 Param 1=20
    Module.Param0WriteEnabled=true
    Module.Param0WriteEnabled1=true
    Module.Param1WriteEnabled=true
    Module.Param1WriteEnabled1=true
    Module.Param2WriteEnabled=true
    Module.Param2WriteEnabled1=true
    Module.Param3WriteEnabled=true
    Module.Param3WriteEnabled1=true
    Write Parameter Index 1=true
   ) {module=DynamicMaterialParameters001}
  ]
  rendererClasses [&NiagaraMeshRendererProperties]
  renderers [
   {rendererClass=&NiagaraMeshRendererProperties, rendererIndex=0}
   {rendererClass=&NiagaraMeshRendererProperties, rendererIndex=1}
  ]
 }
 Big_Old_Flares {
  emitterUpdateScript [
   &EmitterState()
   &SpawnRate(SpawnRate=20, Use Spawn Probability=false)
  ]
  particleSpawnScript [
   &PartitionParticles()
   /Niagara/Modules/Spawn/Initialization/V2/InitializeParticle(
    Color=&MakeLinearColorFromVectorAndFloat(
     "Vector (RGB)"=&MakeVectorFromLinearColorRGB(Color=<rgba 0 0.3333 1 1>)
     "Float (Alpha)"=&FN_FrontEndFloatModifier(Gameplay=0.2, Front End=0)
    )
    Lifetime=&FN_FrontEndFloatModifier(Gameplay=0.2, Front End=0.4)
    Sprite Size Mode=::NewEnumerator3
    Uniform Sprite Size=130
    UsePositionOffset=true
    Write Material Random=true
   )
  ]
  particleUpdateScript [
   &ParticleState()
   &SkeletalMeshLocation(
    Bone / Socket Index=$Particles.PartitionParticles.Partition
    MaskSpawnGroup=true
    Mesh Sampling Type=::NewEnumerator5
    Orientation Sampling=::NewEnumerator1
    Sampled Position Offset=<-2 0 0>
    Skeletal Mesh=$User.Skeletal Mesh
    Socket Sampling Mode=::NewEnumerator5
   )
   &ScaleSpriteSize(
    Scale Sprite Size Mode=::NewEnumerator0
    Uniform Scale Factor=&MakeFloatFromVector2d(Vector 2D=&Multiply_Vector2D_ByFloat(
      Float=0.5
      Vector2D=&Vector2DFromFloat(Value=&FN_FrontEndFloatModifier(Gameplay=$User.ThrusterPower, Front End=0.5))
     ))
   )
   &CameraOffset(Camera Offset Amount=&FN_FrontEndFloatModifier(Gameplay=10, Front End=5))
  ]
  rendererClasses [&NiagaraSpriteRendererProperties]
  renderers [{rendererClass=&NiagaraSpriteRendererProperties}]
 }
 Big_Old_Flares001 {
  emitterUpdateScript [
   &EmitterState(
    Life Cycle Mode=::NewEnumerator1
    Loop Duration=&RandomRangeFloat(Minimum=0.1, Maximum=8)
    Recalculate Duration Each Loop=true
    UseLoopDelay=false
   )
   &SpawnRate(SpawnRate=50, Use Spawn Probability=false)
  ]
  particleSpawnScript [
   &PartitionParticles()
   &InitializeParticle(
    Color=&MakeLinearColorFromVectorAndFloat(
     "Vector (RGB)"=&MakeVectorFromLinearColorRGB(Color=<rgba 0 0.1667 1 1>)
     "Float (Alpha)"=&FloatFromCurve(
      FloatCurve=DataInterface() {propertyValues {Curve {keys [
          {interpMode=RCIM_Cubic, tangentMode=RCTM_User, time=0, value=0.2}
          {interpMode=RCIM_Cubic, tangentMode=RCTM_User, time=0.5, value=0.5}
          {tangentMode=RCTM_User, time=1, value=0.2}
         ]}}}
      CurveIndex=$Emitter.LoopedAge
      Scale Curve=0.2
     )
    )
    Lifetime=0.2
    Sprite Size Mode=::NewEnumerator3
    Uniform Sprite Size=&FloatFromCurve(
     FloatCurve=DataInterface() {propertyValues {Curve {keys [
         {interpMode=RCIM_Cubic, tangentMode=RCTM_User, time=0, value=0.8}
         {interpMode=RCIM_Cubic, tangentMode=RCTM_User, time=0.5, value=1}
         {tangentMode=RCTM_User, time=1, value=0.8}
        ]}}}
     CurveIndex=$Emitter.LoopedAge
     Scale Curve=60
    )
    UsePositionOffset=true
    Write Material Random=true
   )
   /SetVariables(Particles.Color_Particle_B:LinearColor=&MakeLinearColorFromVectorAndFloat("Vector (RGB)"=<1 1 1>, "Float (Alpha)"=&RandomRangeFloat(Minimum=0.6, Maximum=1))) {isSetParameters=true}
  ]
  particleUpdateScript [
   &ParticleState()
   &SkeletalMeshLocation(
    Bone / Socket Index=$Particles.PartitionParticles.Partition
    MaskSpawnGroup=true
    Mesh Sampling Type=::NewEnumerator5
    Orientation Sampling=::NewEnumerator1
    Sampled Position Offset=<-5 0 0>
    Skeletal Mesh=$User.Skeletal Mesh
    Socket Sampling Mode=::NewEnumerator5
   )
   /SetVariables(
    Particles.Color_Particle_B:LinearColor=&MultiplyLinearColor(
     Linear Color=$Particles.Color_Particle_B
     Scale Factor=&MakeVector4_FromLinearColor(LinearColor=<rgba 0 3.3333 20 0.05>)
    )
    Particles.SpriteSizeAlt:Vector2f=&Multiply_Vector2D(A=$Particles.Initial.SpriteSize, B=<0.4 0.4>)
   ) {isSetParameters=true}
   &CameraOffset(Camera Offset Amount=5)
  ]
  rendererClasses [&NiagaraSpriteRendererProperties]
  renderers [
   {rendererClass=&NiagaraSpriteRendererProperties}
   {rendererClass=&NiagaraSpriteRendererProperties, rendererIndex=1}
  ]
 }
 Light {
  emitterUpdateScript [
   &EmitterState()
   &SpawnBurst_Instantaneous(Spawn Count=2, Use Spawn Probability=false, UseLoopCountLimit=true)
  ]
  particleSpawnScript [
   &PartitionParticles()
   /Niagara/Modules/Spawn/Initialization/InitializeParticle(
    Color=$System.SysColorB
    Module.Write Color=true
    Module.Write Lifetime=true
    Module.Write Mass=true
    Module.Write Material Random=true
    Module.Write Position=false
    Module.Write Scale=false
    Module.Write SpriteRotation=false
    Module.Write SpriteSize=true
    Position=&SimulationPosition()
    Sprite Size=<32 32>
   )
  ]
  particleUpdateScript [
   &ParticleState(
    Kill Particles When Lifetime Has Elapsed=false
    Let Infinitely Lived Particles Die When Emitter Deactivates=true
    Loop Particles Lifetime=true
   )
   &SkeletalMeshLocation(
    Bone / Socket Index=$Particles.PartitionParticles.Partition
    MaskSpawnGroup=false
    Mesh Sampling Type=::NewEnumerator5
    Sampled Position Offset=&Add_Vector(A=$System.OffsetLoopFromEngineBone, B=<-1 0 -4>)
    Skeletal Mesh=$User.Skeletal Mesh
    Socket Sampling Mode=::NewEnumerator5
   )
   &FN_Light_v3(
    Color=<rgba 0 3.3333 20 1>
    Create Color Attribute=true
    Create Exponent Attribute=true
    Create Radius Attribute=true
    Enable Particle Light Enable=true
    Exponent=1
    Light Enabled=true
    Radius:NiagaraFloat=&Lerp_Float(
     A=0
     B=&RandomRangeFloat(Minimum=19, Maximum=22, Evaluation Type=::NewEnumerator1)
     Alpha=&FN_FrontEndFloatModifier(Gameplay=$System.ThrusterPower, Front End=1.3)
    )
    Scale Color:NiagaraFloat=&FloatFromCurve(
     FloatCurve=DataInterface() {propertyValues {Curve {keys [<at 0 0.5 RCIM_Cubic>,<at 0.5 1 RCIM_Cubic>,<at 1 0.5>]}}}
     CurveIndex=$Emitter.NormalizedLoopAge
     Scale Curve=&RandomRangeFloat(
      Minimum=40
      Maximum=50
      Evaluation Type=::NewEnumerator1
      Randomness Mode=::NewEnumerator1
      Random Seed=&Add_Integer(A=$Particles.SpawnGroup, B=&Add_Integer(A=$Emitter.LoopCount, B=$System.LoopCount))
      Fixed Random Seed=true
     )
    )
   )
   &ScaleColor(
    Scale Alpha=&Multiply_Float(
     A=&FloatFromCurve(
      FloatCurve=DataInterface() {propertyValues {Curve {keys [{tangentMode=RCTM_User, time=0, value=-0.5}, {tangentMode=RCTM_User, time=1, value=1}]}}}
      CurveIndex=$System.ThrusterPower
     )
     B=&FloatFromCurve(
      FloatCurve=DataInterface() {propertyValues {Curve {keys [
          {interpMode=RCIM_Cubic, tangentMode=RCTM_User, time=0, value=0.5}
          {interpMode=RCIM_Cubic, tangentMode=RCTM_User, time=0.5, value=1}
          {tangentMode=RCTM_User, time=1, value=0.5}
         ]}}}
      CurveIndex=$System.LoopedAge
      Scale Curve=0.5
     )
    )
    ScaleA=true
    ScaleRGB=true
   )
  ]
  rendererClasses [&NiagaraLightRendererProperties]
  renderers [{rendererClass=&NiagaraLightRendererProperties, rendererIndex=0}]
 }
 MainThruster {
  emitterUpdateScript [
   &EmitterState()
   &SpawnRate(SpawnRate=26, Use Spawn Probability=false)
  ]
  particleSpawnScript [
   &PartitionParticles()
   &InitializeParticle(
    Lifetime=&Lerp_Float(A=0.2, B=0.3, Alpha=&FN_FrontEndFloatModifier(Gameplay=$User.ForwardDot, Front End=0.1))
    Mesh Scale=&Multiply_Vector(
     A=&RandomRangeVector(Minimum=<0.2 0.2 0.06>, Maximum=<0.23 0.23 0.09>)
     B=&Lerp_Vector(
      A=<1 1 1>
      B=&RandomRangeVector(Minimum=<0.5 0.5 1.5>, Maximum=<1 1 1.5>)
      Alpha=&FN_FrontEndFloatModifier(Gameplay=$System.ThrusterPower, Front End=1)
     )
    )
    Module.Write Color=true
    Module.Write Lifetime=true
    Module.Write Mass=false
    Module.Write Material Random=false
    Module.Write Position=false
    Module.Write Scale=true
    Module.Write SpriteRotation=false
    Module.Write SpriteSize=false
    Position=&SimulationPosition()
    Sprite Size=<5 5>
   )
   &DynamicMaterialParameters(
    Index 0 Param 1=&RandomRangeFloat(Minimum=0, Maximum=1)
    Index 0 Param 4=&Lerp_Float(A=0, B=20, Alpha=&FN_FrontEndFloatModifier(Gameplay=$System.ThrusterPower, Front End=0))
    Module.Param0WriteEnabled=true
    Module.Param1WriteEnabled=false
    Module.Param2WriteEnabled=false
    Module.Param3WriteEnabled=true
   )
   &InitialMeshOrientation(
    Mesh Axis To Orient=<0 0 1>
    Mesh Orientation Mode=::NewEnumerator2
    Rotation=&RandomRangeVector(Minimum=<0 0 0>, Maximum=<0 0 1>)
    Use Rotation Vector=true
   )
  ]
  particleUpdateScript [
   &ParticleState()
   &SkeletalMeshLocation(
    Bone / Socket Index=$Particles.PartitionParticles.Partition
    MaskSpawnGroup=false
    Mesh Sampling Type=::NewEnumerator5
    Sampled Position Offset=<-10 0 0>
    Skeletal Mesh=$User.Skeletal Mesh
    Socket Sampling Mode=::NewEnumerator5
   )
   &ScaleMeshSize(Scale Factor=&FN_FrontEndVector3Modifier(
     Gameplay=&VectorFromCurve(
      VectorCurve=DataInterface() {propertyValues {
        XCurve {keys [<at 0 1>,<at 1 1>]}
        YCurve {keys [<at 0 1>,<at 1 1>]}
        ZCurve {keys [<at 0.9 1 RCIM_Cubic>, {arriveTangent=-4.873, interpMode=RCIM_Cubic, leaveTangent=-4.873, tangentMode=RCTM_User, time=1, value=0}]}
       }}
      Scale Curve=&Lerp_Vector(
       A=<0.2 0.2 0.1>
       B=<0.4 0.4 2>
       Alpha=&FN_FrontEndFloatModifier(Gameplay=$System.ThrusterPower, Front End=0.8)
      )
     )
     Front End=<0.4 0.4 3>
    ))
   &ScaleColor(
    Scale Alpha=&FN_FrontEndFloatModifier(
     Gameplay=&Lerp_Float(
      A=1
      B=&Lerp_Float(A=0, B=5, Alpha=&FN_FrontEndFloatModifier(Gameplay=$System.ThrusterPower, Front End=1))
      Alpha=$Particles.NormalizedAge
     )
     Front End=0.2
    )
    Scale RGB=<1 0.5 0.5>
    ScaleA=true
    ScaleRGB=true
   )
  ]
  rendererClasses [&NiagaraMeshRendererProperties]
  renderers [{rendererClass=&NiagaraMeshRendererProperties, rendererIndex=0}]
 }
}
summary {
 systemName=NS_Glider_Noble_Jets
 emitters [
  {emitterName=Light, rendererClasses [&NiagaraLightRendererProperties]}
  {emitterName=AmbientThruster, rendererClasses [&NiagaraMeshRendererProperties]}
  {emitterName=Big_Old_Flares, rendererClasses [&NiagaraSpriteRendererProperties]}
  {emitterName=MainThruster, rendererClasses [&NiagaraMeshRendererProperties]}
  {emitterName=Big_Old_Flares001, rendererClasses [&NiagaraSpriteRendererProperties]}
 ]
 userVariables [
  User.bIsFrontEnd {type=NiagaraBool}
  User.bIsFrontEndPreview {type=NiagaraBool}
  User.BPOpacityMult {type=NiagaraFloat, defaultValue=0, description=""}
  User.ColorA {type=LinearColor, description="", defaultValue {struct=LinearColor, value=<rgba 0 0.3333 2 1>}}
  User.ColorB {type=LinearColor, description="", defaultValue {struct=LinearColor, value=<rgba 0.6 1.0667 2 1>}}
  User.ColorHueShift {type=NiagaraFloat, defaultValue=0, description=""}
  User.CurrentExhaust {type=Vector3f, description="", defaultValue {struct=Vector3f, value=<0 0 0>}}
  User.emissive mult {type=NiagaraFloat, defaultValue=0, description=""}
  User.ForwardBurst {type=NiagaraFloat, defaultValue=1, description=""}
  User.ForwardDot {type=NiagaraFloat, defaultValue=-1, description=""}
  User.PlayerSpeed {type=NiagaraFloat, defaultValue=0, description=""}
  User.RightDot {type=NiagaraFloat, defaultValue=0, description=""}
  User.Skeletal Mesh {
   type=NiagaraDataInterfaceSkeletalMesh
   defaultValue {
    struct=NiagaraExt_VariableValue_DataInterface
    value {dataInterface=&self:NiagaraDataInterfaceSkeletalMesh_2}
   }
  }
  User.ThrusterPower {type=NiagaraFloat, defaultValue=0.35, description=""}
  User.TrailLengthScale {type=NiagaraFloat, defaultValue=0, description=""}
 ]
}
